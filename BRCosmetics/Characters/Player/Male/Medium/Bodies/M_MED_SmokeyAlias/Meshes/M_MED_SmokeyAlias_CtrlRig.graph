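# Control Rig for the SmokeyAlias male medium cosmetic body, procedurally posing its collar, jacket, belt straps, sneaker tongues, kneecap, bow and clock bones.
# A shared DeformRig function runs first, then sequenced branches apply weighted ModifyTransforms offsets and parent constraints.
# Weights come from spherical pose readers, remapped joint angles and pelvis height, some smoothed with Kalman or interpolation filters.
# The skydive, DBNO and isLod2 variables select alternate corrections.
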ControlRigBlueprint /BRCosmetics/Characters/Player/Male/Medium/Bodies/M_MED_SmokeyAlias/Meshes/M_MED_SmokeyAlias_CtrlRig
graphs [
 Multiply_1_2_1_ContainedGraph {nodes [
   n1 = RigVMFunctionEntryNode()
   RigVMFunctionReturnNode(Result=&RigVMFunction_MathDoubleMul(B=@n1.C, A=&RigVMFunction_MathDoubleMul(A=@n1.A, B=@n1.B).Result).Result)
  ]}
 RigVMFunctionLibrary {links [], nodes []}
 RigVMFunction_Sequence_1_1_ContainedGraph {nodes [
   n1 = RigVMFunctionEntryNode()
   n2 = &RigVMFunction_Sequence(ExecuteContext=@n1.ExecuteContext)
   n3 = &RigVMFunction_Sequence(ExecuteContext=@n2.B)
   n4 = &RigVMFunction_Sequence(ExecuteContext=@n3.B)
   n5 = &RigVMFunction_Sequence(ExecuteContext=@n4.B)
   n6 = &RigVMFunction_Sequence(ExecuteContext=@n5.B)
   n7 = &RigVMFunction_Sequence(ExecuteContext=@n6.B)
   RigVMFunctionReturnNode(A=@n2.A, B=@n3.A, C=@n4.A, D=@n5.A, H=@n7.A, I=@n7.B, F=@n6.A)
  ]}
 RigVMFunction_Sequence_2_1_1_ContainedGraph {nodes [
   n1 = RigVMFunctionEntryNode()
   n2 = &RigVMFunction_Sequence(ExecuteContext=@n1.ExecuteContext)
   n3 = &RigVMFunction_Sequence(ExecuteContext=@n2.B)
   RigVMFunctionReturnNode(A=@n2.A, B=@n3.A, C=@n3.B)
  ]}
 RigVMModel {nodes [
   &RigUnit_ParentConstraint(
    AdvancedSettings=(InterpolationType=Average,RotationOrderForFilter=XZY)
    Child=(Type=Bone,Name="kneecap")
    ExecutePin=()
    Filter=(TranslationFilter=(bX=false,bY=false,bZ=false),RotationFilter=(bX=True,bY=True,bZ=True),ScaleFilter=(bX=false,bY=false,bZ=false))
    Parents=((Item=(Type=Bone,Name="calf_l"),Weight=0.8),(Item=(Type=Bone,Name="thigh_l"),Weight=0.2))
    Weight=1
   )
   > n1 = &RigUnit_SphericalPoseReader(
    ActiveRegionScaleFactors=(PositiveWidth=1,NegativeWidth=0,PositiveHeight=1,NegativeHeight=1)
    ActiveRegionSize=0.3
    Debug=(bDrawDebug=false,bDraw2D=false,bDrawLocalAxes=false,DebugScale=30,DebugSegments=30,DebugThickness=0.2)
    DriverAxis=(X=-1,Y=0,Z=0)
    DriverItem=(Type=Bone,Name="thigh_l")
    FalloffRegionScaleFactors=(PositiveWidth=1,NegativeWidth=1,PositiveHeight=1,NegativeHeight=1)
    FalloffSize=0.5
    OptionalParentItem=(Type=Bone,Name="pelvis")
    RotationOffset=(X=90,Y=0,Z=0)
   )
   > RigUnit_SetRelativeTransformForItem = /Script/ControlRig.RigUnit_SetRelativeTransformForItem(
    Child=(Type=Bone,Name="Belt_front_l_strap")
    Parent=(Type=Bone,Name="Belt_front_l_strap")
    Value=(Rotation=(X=0.4777,Y=0.3799,Z=0.5964,W=0.5213),Translation=(X=-5.8509,Y=4.4987,Z=-4.8281),Scale3D=(X=1,Y=1,Z=1))
    bParentInitial=False
    bPropagateToChildren=True
    Weight=@n1.OutputParam
   ) {
    is=struct RigVMUnitNode
    template="Set Relative Transform::Execute(in Child,in Parent,in bParentInitial,in Value,in Weight,in bPropagateToChildren)"
   }
   KalmanFilter = &RigVMFunction_KalmanFloat(BufferSize=7, Value=@n1.OutputParam) {template="KalmanFilter::Execute(in Value,in BufferSize,out Result)"}
   > &RigUnit_ModifyTransforms(
    ItemToModify=((Item=(Type=Bone,Name="kneecap"),Transform=(Rotation=(X=0.2759,Y=0,Z=0,W=0.9612),Translation=(X=0,Y=-2.1796,Z=-1.3066),Scale3D=(X=1,Y=1,Z=1))))
    WeightMinimum=0
    Weight=@KalmanFilter.Result
   )
   Multiply_1_2 = RigVMAggregateNode(
    A=@KalmanFilter.Result
    B=&RigVMFunction_MathDoubleRemap(
     SourceMaximum=45
     SourceMinimum=70
     TargetMaximum=1
     TargetMinimum=0
     Value=&RigUnit_GetRelativeTransformForItem(Child=(Type=Bone,Name="pelvis"), Parent=(Type=Bone,Name="root"), bChildInitial=False, bParentInitial=False).RelativeTransform.Translation.Z
    ).Result
    C=&RigVMFunction_MathDoubleRemap(
     SourceMaximum=40
     SourceMinimum=35
     TargetMaximum=1
     TargetMinimum=0
     Value=&RigUnit_GetRelativeTransformForItem(Child=(Type=Bone,Name="pelvis"), Parent=(Type=Bone,Name="root"), bChildInitial=False, bParentInitial=False).RelativeTransform.Translation.Z
    ).Result
   )
   > &RigUnit_ModifyTransforms(
    ItemToModify="((Item=(Type=Bone,Name=\"Jacket_base_l\"),Transform=(Rotation=(X=0,Y=0.1736,Z=0,W=0.9848),Translation=(X=-7,Y=0,Z=0),Scale3D=(X=1,Y=1,Z=1))),(Item=(Type=Bone,Name=\"Jacket_base_r\"),Transform=(Rotation=(X=-0.1285,Y=0.1722,Z=0.0227,W=0.9764),Translation=(X=0,Y= ...(truncated)"
    WeightMinimum=0
    Weight=@Multiply_1_2.Result
   )
   RigUnit_BeginExecution = &RigUnit_BeginExecution()
   FN_DeformRig_FNC = RigVMFunctionReferenceNode(
    DeformRig_ Curve="(deformRig_l_upperarm_up_alpha_72_#1=False,deformRig_r_upperarm_up_alpha_73_#2=False,deformRig_l_upperarm_fr_alpha_74_#3=False,deformRig_r_upperarm_fr_alpha_75_CB51B8 ...(truncated)"
    DeformRig_Bool="(upperarm_1_#4=True,elbow_3_#5=True,wrist_7_#6=True,groin_8_#7=True,glute_17_#8=True,knee_9_40C408324CA261 ...(truncated)"
    DeformRig_Debug=(upperarm_debug_18_#4=False,wrist_debug_20_#6=False,lat_debug_26_#9=False,groinGlute_debug_27_#7=False)
    DeformRig_Value="(upperarm_up_18_#10=(Rotation=(X=0,Y=0,Z=0,W=1),Translation=(X=0,Y=0,Z=8),Scale3D=(X=0.7,Y=1,Z=1)),upperarm_fwd_fr_19_#11=(Rotation=(X=0,Y=0,Z=0,W=1),Translation=(X=0,Y=10,Z=0),Scale3D=(X=0.5,Y=1,Z= ...(truncated)"
    ExecuteContext=@RigUnit_BeginExecution.ExecutePin
   )
   > RigVMFunction_Sequence_1 = RigVMAggregateNode()
   n2 = &RigVMFunction_ControlFlowBranch(ExecuteContext=@RigVMFunction_Sequence_1.A, Condition=RigVMVariableNode() {variable=skydive}.Value)
   n3 = &RigUnit_SphericalPoseReader(
    ActiveRegionScaleFactors=(PositiveWidth=1,NegativeWidth=1.1,PositiveHeight=0.75,NegativeHeight=1.5)
    ActiveRegionSize=0.08
    Debug=(bDrawDebug=false,bDraw2D=false,bDrawLocalAxes=false,DebugScale=30,DebugSegments=30,DebugThickness=0.2)
    DriverAxis=(X=1,Y=0,Z=0)
    DriverItem=(Type=Bone,Name="head")
    FalloffRegionScaleFactors=(PositiveWidth=1.2,NegativeWidth=1.2,PositiveHeight=1.6,NegativeHeight=1)
    FalloffSize=0.15
    OptionalParentItem=(Type=Bone,Name="pelvis")
    RotationOffset=(X=-50,Y=-70,Z=0)
    ExecutePin=@n2.False
   )
   > &RigUnit_ModifyTransforms(
    ItemToModify="((Item=(Type=Bone,Name=\"Collar_base_right_ctrl\"),Transform=(Rotation=(X=0.4897,Y=-0.1932,Z=0.1117,W=0.8428),Translation=(X=0,Y=0,Z=0),Scale3D=(X=1,Y=1,Z=1))),(Item=(Type=Bone,Name=\"Collar_base_midle_ctrl\"),Transform=(Rotation=(X=0.3608,Y=-0.1365,Z=-0.0646, ...(truncated)"
    WeightMinimum=0
    Weight=@n3.OutputParam
   )
   n4 = &RigUnit_GetRelativeTransformForItem(Child=(Type=Bone,Name="pelvis"), Parent=(Type=Bone,Name="root"), bChildInitial=False, bParentInitial=False)
   > &RigUnit_ModifyTransforms(
    ItemToModify="((Item=(Type=Bone,Name=\"Collar_base_right_ctrl\"),Transform=(Rotation=(X=-0.0004,Y=-0.0879,Z=0.1223,W=0.9886),Translation=(X=0,Y=0,Z=0),Scale3D=(X=1,Y=1,Z=1))),(Item=(Type=Bone,Name=\"Collar_base_midle_ctrl\"),Transform=(Rotation=(X=0.043,Y=-0.1735,Z=0.0076,W ...(truncated)"
    WeightMinimum=0
    Weight=&RigVMFunction_MathDoubleMul(
     B=&RigVMFunction_MathDoubleRemap(SourceMaximum=50, SourceMinimum=80, TargetMaximum=1, TargetMinimum=0, Value=@n4.RelativeTransform.Translation.Z).Result
     A=@n3.OutputParam
    ).Result
   )
   n5 = &RigVMFunction_MathDoubleRemap(SourceMaximum=69, SourceMinimum=50, TargetMaximum=1, TargetMinimum=0, Value=@n4.RelativeTransform.Translation.Z)
   n6 = &RigVMFunction_MathDoubleRemap(SourceMaximum=67, SourceMinimum=77, TargetMaximum=1, TargetMinimum=0, Value=@n4.RelativeTransform.Translation.Z)
   > &RigUnit_ModifyTransforms(
    ItemToModify="((Item=(Type=Bone,Name=\"Collar_base_right_ctrl\"),Transform=(Rotation=(X=0.2549,Y=0.1677,Z=-0.0449,W=0.9513),Translation=(X=-5.1725,Y=-1.6105,Z=-3.4776),Scale3D=(X=1,Y=1,Z=1))),(Item=(Type=Bone,Name=\"Collar_base_midle_ctrl\"),Transform=(Rotation=(X=0.4617,Y= ...(truncated)"
    WeightMinimum=0
    Weight=&RigVMFunction_KalmanFloat(BufferSize=50, Value=&RigVMFunction_MathDoubleMul(A=@n5.Result, B=@n6.Result).Result) {template="KalmanFilter::Execute(in Value,in BufferSize,out Result)"}.Result
   )
   > &RigUnit_ModifyTransforms(
    ItemToModify="((Item=(Type=Bone,Name=\"Collar_base_right_ctrl\"),Transform=(Rotation=(X=-0.1508,Y=0.2318,Z=-0.059,W=0.9592),Translation=(X=0,Y=0,Z=0),Scale3D=(X=1,Y=1,Z=1))),(Item=(Type=Bone,Name=\"Collar_base_midle_ctrl\"),Transform=(Rotation=(X=0.049,Y=0.2188,Z=0.013,W=0. ...(truncated)"
    WeightMinimum=0
    Weight=&RigVMFunction_MathDoubleMul(B=@n5.Result, A=@n3.OutputParam).Result
   )
   n7 = &RigUnit_GetRelativeTransformForItem(Child=(Type=Bone,Name="head"), Parent=(Type=Bone,Name="spine_05"), bChildInitial=False, bParentInitial=False)
   AlphaInterp = &RigVMFunction_AlphaInterp(
    InterpSpeedDecreasing=10
    InterpSpeedIncreasing=10
    Scale=1
    bClampResult=False
    Value=&RigVMFunction_MathDoubleRemap(
     SourceMaximum=-40
     SourceMinimum=-10
     TargetMaximum=1
     TargetMinimum=0
     Value=&RigVMFunction_MathQuaternionToEuler(RotationOrder=ZYX, Value=@n7.RelativeTransform.Rotation).Result.Z
    ).Result
   )
   > &RigUnit_ModifyTransforms(
    ItemToModify=((Item=(Type=Bone,Name="Collar_base_black_ctrl"),Transform=(Rotation=(X=0.1736,Y=0,Z=0,W=0.9848),Translation=(X=-1,Y=0,Z=-1.3),Scale3D=(X=1,Y=1,Z=1))))
    WeightMinimum=0
    Weight=@AlphaInterp.Result
   )
   > &RigUnit_ModifyTransforms(
    ItemToModify=((Item=(Type=Bone,Name="Collar_base_black_ctrl"),Transform=(Rotation=(X=0,Y=0,Z=0,W=1),Translation=(X=0,Y=1,Z=-2),Scale3D=(X=1,Y=1,Z=1))))
    WeightMinimum=0
    Weight=@n3.OutputParam
   )
   n8 = &RigUnit_GetRelativeTransformForItem(Child=(Type=Bone,Name="spine_05"), Parent=(Type=Bone,Name="pelvis"), bChildInitial=False, bParentInitial=False)
   > &RigUnit_ModifyTransforms(
    ItemToModify=((Item=(Type=Bone,Name="Collar_base_black_ctrl"),Transform=(Rotation=(X=0,Y=-0.1564,Z=0,W=0.9877),Translation=(X=0,Y=1,Z=1),Scale3D=(X=1,Y=1,Z=1))))
    WeightMinimum=0
    Weight=&RigVMFunction_MathDoubleRemap(
     SourceMaximum=30
     SourceMinimum=20
     TargetMaximum=1
     TargetMinimum=0
     Value=&RigVMFunction_MathQuaternionToEuler(RotationOrder=ZYX, Value=@n8.RelativeTransform.Rotation).Result.X
    ).Result
   )
   n9 = &RigVMFunction_MathQuaternionToEuler(
    RotationOrder=ZYX
    Value=&RigUnit_GetRelativeTransformForItem(Child=(Type=Bone,Name="head"), Parent=(Type=Bone,Name="neck_02"), bChildInitial=False, bParentInitial=False).RelativeTransform.Rotation
   )
   > &RigUnit_ModifyTransforms(
    ItemToModify="((Item=(Type=Bone,Name=\"Collar_l\"),Transform=(Rotation=(X=0,Y=-0.3827,Z=0,W=0.9239),Translation=(X=0,Y=0,Z=0),Scale3D=(X=1,Y=1,Z=1))),(Item=(Type=Bone,Name=\"Collar_black_l\"),Transform=(Rotation=(X=0,Y=-0.1305,Z=0,W=0.9914),Translation=(X=0,Y=0,Z=-3),Scale3 ...(truncated)"
    WeightMinimum=0
    Weight=&RigVMFunction_MathDoubleRemap(SourceMaximum=-60, SourceMinimum=0, TargetMaximum=1, TargetMinimum=0, Value=@n9.Result.X).Result
   )
   > &RigUnit_ModifyTransforms(
    ItemToModify="((Item=(Type=Bone,Name=\"Collar_r\"),Transform=(Rotation=(X=0,Y=-0.3827,Z=0,W=0.9239),Translation=(X=0,Y=0,Z=0),Scale3D=(X=1,Y=1,Z=1))),(Item=(Type=Bone,Name=\"Collar_black_r\"),Transform=(Rotation=(X=0,Y=-0.0872,Z=0,W=0.9962),Translation=(X=0,Y=0,Z=-3),Scale3 ...(truncated)"
    WeightMinimum=0
    Weight=&RigVMFunction_MathDoubleRemap(SourceMaximum=60, SourceMinimum=0, TargetMaximum=1, TargetMinimum=0, Value=@n9.Result.X).Result
   )
   n10 = &RigUnit_SphericalPoseReader(
    ActiveRegionScaleFactors=(PositiveWidth=1,NegativeWidth=1.1,PositiveHeight=1,NegativeHeight=1.5)
    ActiveRegionSize=0.22
    Debug=(bDrawDebug=false,bDraw2D=false,bDrawLocalAxes=false,DebugScale=30,DebugSegments=30,DebugThickness=0.2)
    DriverAxis=(X=1,Y=0,Z=0)
    DriverItem=(Type=Bone,Name="head")
    FalloffRegionScaleFactors=(PositiveWidth=1,NegativeWidth=1,PositiveHeight=1,NegativeHeight=1)
    FalloffSize=0.35
    OptionalParentItem=(Type=Bone,Name="pelvis")
    RotationOffset=(X=-85,Y=-100,Z=0)
    ExecutePin=@n2.True
   )
   > &RigUnit_ModifyTransforms(
    ItemToModify="((Item=(Type=Bone,Name=\"Collar_base_right_ctrl\"),Transform=(Rotation=(X=0.2578,Y=0.0842,Z=-0.0226,W=0.9623),Translation=(X=0,Y=0,Z=0),Scale3D=(X=1,Y=1,Z=1))),(Item=(Type=Bone,Name=\"Collar_base_midle_ctrl\"),Transform=(Rotation=(X=0.454,Y=0,Z=0,W=0.891),Tran ...(truncated)"
    WeightMinimum=0
    Weight=@n10.OutputParam
   )
   n11 = &RigUnit_GetRelativeTransformForItem(Child=(Type=Bone,Name="spine_05"), Parent=(Type=Bone,Name="pelvis"), bChildInitial=False, bParentInitial=False)
   > &RigUnit_ModifyTransforms(
    ItemToModify="((Item=(Type=Bone,Name=\"None\"),Transform=(Rotation=(X=0,Y=0,Z=0,W=1),Translation=(X=0,Y=0,Z=0),Scale3D=(X=1,Y=1,Z=1))),(Item=(Type=Bone,Name=\"Belt_front_l_strap\"),Transform=(Rotation=(X=0,Y=0,Z=0,W=1),Translation=(X=2,Y=0,Z=-1.5),Scale3D=(X=1,Y=1,Z=1))),(I ...(truncated)"
    WeightMinimum=0
    Weight=&RigVMFunction_MathDoubleRemap(
     SourceMaximum=-19
     SourceMinimum=-35
     TargetMaximum=1
     TargetMinimum=0
     Value=&RigVMFunction_MathQuaternionToEuler(RotationOrder=ZYX, Value=@n11.RelativeTransform.Rotation).Result.X
    ).Result
   )
   RigVMFunction_Sequence_2_1 = RigVMAggregateNode(ExecuteContext=@RigVMFunction_Sequence_1.B)
   RigVMFunction_ControlFlowBranch = &RigVMFunction_ControlFlowBranch(ExecuteContext=@RigVMFunction_Sequence_2_1.A, Condition=RigVMVariableNode() {variable=DBNO}.Value)
   n12 = &RigUnit_SphericalPoseReader(
    ActiveRegionScaleFactors=(PositiveWidth=1,NegativeWidth=1,PositiveHeight=1,NegativeHeight=1)
    ActiveRegionSize=0.06
    Debug=(bDrawDebug=false,bDraw2D=false,bDrawLocalAxes=false,DebugScale=30,DebugSegments=30,DebugThickness=0.2)
    DriverAxis=(X=1,Y=0,Z=0)
    DriverItem=(Type=Bone,Name="thigh_r")
    FalloffRegionScaleFactors=(PositiveWidth=1,NegativeWidth=1,PositiveHeight=1,NegativeHeight=1)
    FalloffSize=0.13
    OptionalParentItem=(Type=Bone,Name="pelvis")
    RotationOffset=(X=-60,Y=15,Z=0)
    ExecutePin=@RigVMFunction_ControlFlowBranch.False
   )
   > &RigUnit_ModifyTransforms(
    ItemToModify=((Item=(Type=Bone,Name="Jacket_base_r"),Transform=(Rotation=(X=0,Y=0.3827,Z=0,W=0.9239),Translation=(X=0,Y=0,Z=0),Scale3D=(X=1,Y=1,Z=1))))
    WeightMinimum=0
    Weight=@n12.OutputParam
   )
   &RigUnit_ModifyTransforms(
    ItemToModify="((Item=(Type=Bone,Name=\"Belt_front_l_strap\"),Transform=(Rotation=(X=-0.4588,Y=0.2336,Z=-0.0626,W=0.855),Translation=(X=-1,Y=-1,Z=0),Scale3D=(X=1,Y=1,Z=1))),(Item=(Type=Bone,Name=\"Belt_front_r_strap\"),Transform=(Rotation=(X=0.342,Y=0,Z=0,W=0.9397),Translati ...(truncated)"
    Weight=1
    WeightMinimum=0
    ExecutePin=@RigVMFunction_ControlFlowBranch.True
   )
   n13 = &RigUnit_GetRelativeTransformForItem(Child=(Type=Bone,Name="spine_05"), Parent=(Type=Bone,Name="pelvis"), bChildInitial=False, bParentInitial=False)
   &RigUnit_ModifyTransforms(
    ItemToModify=((Item=(Type=Bone,Name="Belt_front_l_strap"),Transform=(Rotation=(X=0,Y=0,Z=0,W=1),Translation=(X=-2.4,Y=0,Z=0.5),Scale3D=(X=1,Y=1,Z=1))))
    WeightMinimum=0
    Weight=&RigVMFunction_MathDoubleRemap(
     SourceMaximum=30
     SourceMinimum=18
     TargetMaximum=1
     TargetMinimum=0
     Value=&RigVMFunction_MathQuaternionToEuler(RotationOrder=ZYX, Value=@n13.RelativeTransform.Rotation).Result.Y
    ).Result
    ExecutePin=@RigVMFunction_Sequence_2_1.B
   )
   n14 = &RigUnit_GetRelativeTransformForItem(Child=(Type=Bone,Name="pelvis"), Parent=(Type=Bone,Name="root"), bChildInitial=False, bParentInitial=False)
   n15 = &RigVMFunction_MathQuaternionToEuler(
    RotationOrder=ZYX
    Value=&RigUnit_GetRelativeTransformForItem(Child=(Type=Bone,Name="spine_05"), Parent=(Type=Bone,Name="pelvis"), bChildInitial=False, bParentInitial=False).RelativeTransform.Rotation
   )
   RigVMFunction_AnimEvalRichCurve = &RigVMFunction_AnimEvalRichCurve(
    Curve=(EditorCurveData=(Keys=((InterpMode=RCIM_Cubic,TangentMode=RCTM_Break,LeaveTangent=2.5188),(InterpMode=RCIM_Cubic,TangentMode=RCTM_Break,Time=0.2295,Value=0.6312,ArriveTangent=0.6149,LeaveTangent=0.6265),(Time=1,Value=1))))
    SourceMaximum=30
    SourceMinimum=10
    TargetMinimum=0
    Value=@n15.Result.Y
   )
   > &RigUnit_ModifyTransforms(
    ItemToModify="((Item=(Type=Bone,Name=\"Jacket_base_l\"),Transform=(Rotation=(X=0.619,Y=0.4517,Z=-0.0596,W=0.6397),Translation=(X=0,Y=0,Z=-1),Scale3D=(X=1,Y=1,Z=1))),(Item=(Type=Bone,Name=\"Jacketback_l_base\"),Transform=(Rotation=(X=0.0858,Y=0.173,Z=-0.0151,W=0.9811),Transl ...(truncated)"
    WeightMinimum=0
    Weight=&RigVMFunction_MathDoubleMul(
     A=&RigVMFunction_MathDoubleRemap(SourceMaximum=50, SourceMinimum=80, TargetMaximum=1, TargetMinimum=0, Value=@n14.RelativeTransform.Translation.Z).Result
     B=@RigVMFunction_AnimEvalRichCurve.Result
    ).Result
   )
   n16 = &RigVMFunction_MathDoubleRemap(SourceMaximum=30, SourceMinimum=20, TargetMaximum=1, TargetMinimum=0, Value=@n15.Result.X)
   > &RigUnit_ModifyTransforms(
    ItemToModify=((Item=(Type=Bone,Name="Jacket_base_l"),Transform=(Rotation=(X=-0.0189,Y=0.2156,Z=0.0851,W=0.9726),Translation=(X=0,Y=0,Z=0),Scale3D=(X=1,Y=1,Z=1))))
    WeightMinimum=0
    Weight=@n16.Result
   )
   > n17 = &RigUnit_SphericalPoseReader(
    ActiveRegionScaleFactors=(PositiveWidth=1,NegativeWidth=1,PositiveHeight=1,NegativeHeight=1)
    ActiveRegionSize=0.1
    Debug=(bDrawDebug=false,bDraw2D=False,bDrawLocalAxes=False,DebugScale=25,DebugSegments=20,DebugThickness=0.25)
    DriverAxis=(X=1,Y=0,Z=0)
    DriverItem=(Type=Bone,Name="head")
    FalloffRegionScaleFactors=(PositiveWidth=1,NegativeWidth=1,PositiveHeight=1,NegativeHeight=1)
    FalloffSize=0.2
    FlipHeightScaling=False
    FlipWidthScaling=false
    OptionalParentItem=(Type=Bone,Name="None")
    RotationOffset=(X=0,Y=-80,Z=-25)
   )
   > &RigUnit_ModifyTransforms(
    ItemToModify=((Item=(Type=Bone,Name="Jacketback_l_base"),Transform=(Rotation=(X=0.0114,Y=0.13,Z=-0.0864,W=0.9877),Translation=(X=-2.5,Y=0,Z=-1),Scale3D=(X=1,Y=1,Z=1))))
    WeightMinimum=0
    Weight=&RigVMFunction_MathDoubleMul(A=@n16.Result, B=@n17.OutputParam).Result
   )
   > &RigUnit_ParentConstraint(
    AdvancedSettings=(InterpolationType=Average,RotationOrderForFilter=XZY)
    Child=(Type=Bone,Name="bow")
    Parents="((Item=(Type=Bone,Name=\"pelvis\"),Weight=0.6797),(Item=(Type=Bone,Name=\"spine_01\"),Weight=0.1046),(Item=(Type=Bone,Name=\"spine_02\"),Weight=0.016),(Item=(Type=Bone,Name=\"spine_03\"),Weight=0.0053),(Item=(Type=Bone,Name=\"thigh_l\"),Weight=0.0186),(Item=(Type=Bo ...(truncated)"
    Weight=1
   )
   n18 = &RigVMFunction_MathDoubleRemap(
    SourceMaximum=50
    SourceMinimum=80
    TargetMaximum=1
    TargetMinimum=0
    Value=&RigUnit_GetRelativeTransformForItem(Child=(Type=Bone,Name="pelvis"), Parent=(Type=Bone,Name="root"), bChildInitial=False, bParentInitial=False).RelativeTransform.Translation.Z
   )
   > &RigUnit_ModifyTransforms(
    ItemToModify=((Item=(Type=Bone,Name="bow"),Transform=(Rotation=(X=0,Y=-0.1305,Z=0,W=0.9914),Translation=(X=0,Y=0,Z=0),Scale3D=(X=1,Y=1,Z=1))))
    WeightMinimum=0
    Weight=@n18.Result
   )
   n19 = &RigUnit_GetRelativeTransformForItem(Child=(Type=Bone,Name="calf_l"), Parent=(Type=Bone,Name="thigh_l"), bChildInitial=False, bParentInitial=False)
   > &RigUnit_ModifyTransforms(
    ItemToModify=((Item=(Type=Bone,Name="kneecap"),Transform=(Rotation=(X=0.5,Y=0,Z=0,W=0.866),Translation=(X=0,Y=0,Z=-3.5),Scale3D=(X=1,Y=1,Z=1))))
    WeightMinimum=0
    Weight=&RigVMFunction_MathDoubleRemap(
     SourceMaximum=124
     SourceMinimum=50
     TargetMaximum=1
     TargetMinimum=0
     Value=&RigVMFunction_MathQuaternionToEuler(RotationOrder=ZYX, Value=@n19.RelativeTransform.Rotation).Result.Z
    ).Result
   )
   n20 = &RigVMFunction_MathQuaternionToEuler(
    RotationOrder=ZYX
    Value=&RigUnit_GetRelativeTransformForItem(Child=(Type=Bone,Name="hand_r"), Parent=(Type=Bone,Name="lowerarm_r"), bChildInitial=False, bParentInitial=False).RelativeTransform.Rotation
   )
   &RigUnit_ModifyTransforms(
    ItemToModify=((Item=(Type=Bone,Name="clock"),Transform=(Rotation=(X=0,Y=-0.2588,Z=0,W=0.9659),Translation=(X=-3.5,Y=0,Z=1),Scale3D=(X=1,Y=1,Z=1))))
    WeightMinimum=0
    Weight=&RigVMFunction_MathDoubleRemap(
     SourceMaximum=50
     SourceMinimum=-3
     TargetMaximum=1
     TargetMinimum=0
     Value=&RigVMFunction_KalmanFloat(BufferSize=5, Value=@n20.Result.Z) {template="KalmanFilter::Execute(in Value,in BufferSize,out Result)"}.Result
    ).Result
    ExecutePin=@RigVMFunction_Sequence_2_1.C
   )
   n21 = &RigVMFunction_MathQuaternionToEuler(
    RotationOrder=ZYX
    Value=&RigUnit_GetRelativeTransformForItem(Child=(Type=Bone,Name="hand_r"), Parent=(Type=Bone,Name="lowerarm_r"), bChildInitial=False, bParentInitial=False).RelativeTransform.Rotation
   )
   > &RigUnit_ModifyTransforms(
    ItemToModify=((Item=(Type=Bone,Name="clock"),Transform=(Rotation=(X=0,Y=0,Z=-0.2588,W=0.9659),Translation=(X=0,Y=0,Z=1),Scale3D=(X=1,Y=1,Z=1))))
    WeightMinimum=0
    Weight=&RigVMFunction_MathDoubleRemap(
     SourceMaximum=-75
     SourceMinimum=-3
     TargetMaximum=1
     TargetMinimum=0
     Value=&RigVMFunction_KalmanFloat(BufferSize=5, Value=@n21.Result.Y) {template="KalmanFilter::Execute(in Value,in BufferSize,out Result)"}.Result
    ).Result
   )
   n22 = &RigVMFunction_Sequence(ExecuteContext=@RigVMFunction_Sequence_1.C)
   n23 = &RigUnit_SphericalPoseReader(
    ActiveRegionScaleFactors=(PositiveWidth=1,NegativeWidth=0.1,PositiveHeight=1,NegativeHeight=1)
    ActiveRegionSize=0.3
    Debug=(bDrawDebug=false,bDraw2D=false,bDrawLocalAxes=false,DebugScale=30,DebugSegments=30,DebugThickness=0.2)
    DriverAxis=(X=1,Y=0,Z=0)
    DriverItem=(Type=Bone,Name="thigh_r")
    FalloffRegionScaleFactors=(PositiveWidth=1,NegativeWidth=1.3,PositiveHeight=1,NegativeHeight=1)
    FalloffSize=0.5
    OptionalParentItem=(Type=Bone,Name="pelvis")
    RotationOffset=(X=-90,Y=180,Z=-5)
    ExecutePin=@n22.A
   )
   > &RigUnit_ModifyTransforms(
    ItemToModify="((Item=(Type=Bone,Name=\"Base_Belt_01_backside_r\"),Transform=(Rotation=(X=0,Y=-0.2588,Z=0,W=0.9659),Translation=(X=0,Y=0,Z=0),Scale3D=(X=1,Y=1,Z=1))),(Item=(Type=Bone,Name=\"Belt_front_r_strap\"),Transform=(Rotation=(X=-0.3265,Y=0.6744,Z=-0.2126,W=0.6272),Tra ...(truncated)"
    WeightMinimum=0
    Weight=@n23.OutputParam
   )
   > n24 = &RigUnit_SphericalPoseReader(
    ActiveRegionScaleFactors=(PositiveWidth=0.6,NegativeWidth=1.4,PositiveHeight=1,NegativeHeight=1)
    ActiveRegionSize=0.04
    Debug=(bDrawDebug=false,bDraw2D=false,bDrawLocalAxes=false,DebugScale=30,DebugSegments=30,DebugThickness=0.2)
    DriverAxis=(X=1,Y=0,Z=0)
    DriverItem=(Type=Bone,Name="thigh_r")
    FalloffRegionScaleFactors=(PositiveWidth=1.2,NegativeWidth=1.4,PositiveHeight=1,NegativeHeight=1)
    FalloffSize=0.1
    FlipWidthScaling=true
    OptionalParentItem=(Type=Bone,Name="pelvis")
    RotationOffset=(X=-120,Y=180,Z=-20)
   )
   n25 = &RigVMFunction_AlphaInterp(InterpSpeedDecreasing=14, InterpSpeedIncreasing=14, Scale=1, bClampResult=False, bInterpResult=False, Value=@n24.OutputParam)
   > &RigUnit_ModifyTransforms(
    ItemToModify=((Item=(Type=Bone,Name="Belt_front_r_strap"),Transform=(Rotation=(X=0.1164,Y=-0.0347,Z=0.0041,W=0.9926),Translation=(X=4.8011,Y=-2.89,Z=-2.1433),Scale3D=(X=1,Y=1,Z=1))))
    WeightMinimum=0
    Weight=@n25.Result
   )
   > &RigUnit_ModifyTransforms(
    ItemToModify=((Item=(Type=Bone,Name="Jacket_base_r"),Transform=(Rotation=(X=-0.2294,Y=0.2343,Z=-0.0429,W=0.9437),Translation=(X=0,Y=0,Z=0),Scale3D=(X=1,Y=1,Z=1))))
    WeightMinimum=0
    Weight=@n24.OutputParam
   )
   > n26 = &RigUnit_SphericalPoseReader(
    ActiveRegionScaleFactors=(PositiveWidth=1,NegativeWidth=1,PositiveHeight=1,NegativeHeight=1)
    ActiveRegionSize=0.03
    Debug=(bDrawDebug=false,bDraw2D=false,bDrawLocalAxes=false,DebugScale=30,DebugSegments=30,DebugThickness=0.2)
    DriverAxis=(X=1,Y=0,Z=0)
    DriverItem=(Type=Bone,Name="thigh_r")
    FalloffRegionScaleFactors=(PositiveWidth=1,NegativeWidth=1,PositiveHeight=1,NegativeHeight=1)
    FalloffSize=0.1
    FlipWidthScaling=true
    OptionalParentItem=(Type=Bone,Name="pelvis")
    RotationOffset=(X=-70,Y=0,Z=-32)
   )
   n27 = &RigUnit_GetRelativeTransformForItem(Child=(Type=Bone,Name="pelvis"), Parent=(Type=Bone,Name="root"), bChildInitial=False, bParentInitial=False)
   > &RigUnit_ModifyTransforms(
    ItemToModify=((Item=(Type=Bone,Name="Belt_front_r_strap"),Transform=(Rotation=(X=0,Y=0,Z=0,W=1),Translation=(X=11,Y=0,Z=-1),Scale3D=(X=1,Y=1,Z=1))))
    WeightMinimum=0
    Weight=&RigVMFunction_MathDoubleMul(
     B=&RigVMFunction_MathDoubleRemap(SourceMaximum=85, SourceMinimum=70, TargetMaximum=1, TargetMinimum=0, Value=@n27.RelativeTransform.Translation.Z).Result
     A=@n26.OutputParam
    ).Result
   )
   > n28 = &RigUnit_SphericalPoseReader(
    ActiveRegionScaleFactors=(PositiveWidth=1,NegativeWidth=1,PositiveHeight=1,NegativeHeight=1)
    ActiveRegionSize=0.03
    Debug=(bDrawDebug=false,bDraw2D=false,bDrawLocalAxes=false,DebugScale=30,DebugSegments=30,DebugThickness=0.2)
    DriverAxis=(X=1,Y=0,Z=0)
    DriverItem=(Type=Bone,Name="thigh_r")
    FalloffRegionScaleFactors=(PositiveWidth=1,NegativeWidth=1.3,PositiveHeight=1,NegativeHeight=1)
    FalloffSize=0.1
    FlipWidthScaling=true
    OptionalParentItem=(Type=Bone,Name="pelvis")
    RotationOffset=(X=-45,Y=0,Z=-40)
   )
   > &RigUnit_ModifyTransforms(
    ItemToModify=((Item=(Type=Bone,Name="Belt_front_r_strap"),Transform=(Rotation=(X=0,Y=0,Z=-0.2164,W=0.9763),Translation=(X=0,Y=-10,Z=-10),Scale3D=(X=1,Y=1,Z=1))))
    WeightMinimum=0
    Weight=@n28.OutputParam
   )
   n29 = &RigUnit_SphericalPoseReader(
    ActiveRegionScaleFactors=(PositiveWidth=1,NegativeWidth=0.1,PositiveHeight=1,NegativeHeight=1)
    ActiveRegionSize=0.3
    Debug=(bDrawDebug=false,bDraw2D=false,bDrawLocalAxes=false,DebugScale=30,DebugSegments=30,DebugThickness=0.2)
    DriverAxis=(X=-1,Y=0,Z=0)
    DriverItem=(Type=Bone,Name="thigh_l")
    FalloffRegionScaleFactors=(PositiveWidth=1,NegativeWidth=1.5,PositiveHeight=1,NegativeHeight=1)
    FalloffSize=0.5
    OptionalParentItem=(Type=Bone,Name="pelvis")
    RotationOffset=(X=115,Y=0,Z=-30)
    ExecutePin=@n22.B
   )
   > &RigUnit_ModifyTransforms(
    ItemToModify="((Item=(Type=Bone,Name=\"Base_Belt_01_backside_l\"),Transform=(Rotation=(X=0,Y=-0.2588,Z=0,W=0.9659),Translation=(X=0,Y=0,Z=0),Scale3D=(X=1,Y=1,Z=1))),(Item=(Type=Bone,Name=\"Belt_front_l_strap\"),Transform=(Rotation=(X=0.3287,Y=0.6467,Z=0.1189,W=0.678),Transl ...(truncated)"
    WeightMinimum=0
    Weight=@n29.OutputParam
   )
   > &RigUnit_ModifyTransforms(
    ItemToModify=((Item=(Type=Bone,Name="Belt_front_l_strap"),Transform=(Rotation=(X=0,Y=0,Z=0.3827,W=0.9239),Translation=(X=0.1326,Y=0.3446,Z=1.4305),Scale3D=(X=1,Y=1,Z=1))))
    WeightMinimum=0
    Weight=&RigVMFunction_MathDoubleRemap(
     SourceMaximum=50
     SourceMinimum=80
     TargetMaximum=1
     TargetMinimum=0
     Value=&RigUnit_GetRelativeTransformForItem(Child=(Type=Bone,Name="pelvis"), Parent=(Type=Bone,Name="root"), bChildInitial=False, bParentInitial=False).RelativeTransform.Translation.Z
    ).Result
   )
   n30 = &RigUnit_SphericalPoseReader(
    ActiveRegionScaleFactors=(PositiveWidth=1,NegativeWidth=1,PositiveHeight=1,NegativeHeight=1)
    ActiveRegionSize=0.1
    Debug=(bDrawDebug=false,bDraw2D=false,bDrawLocalAxes=false,DebugScale=30,DebugSegments=30,DebugThickness=0.2)
    DriverAxis=(X=-1,Y=0,Z=0)
    DriverItem=(Type=Bone,Name="thigh_l")
    FalloffRegionScaleFactors=(PositiveWidth=1,NegativeWidth=1,PositiveHeight=1,NegativeHeight=1)
    FalloffSize=0.15
    OptionalParentItem=(Type=Bone,Name="pelvis")
    RotationOffset=(X=110,Y=0,Z=35)
    ExecutePin=@RigVMFunction_Sequence_1.D
   )
   n31 = &RigUnit_GetRelativeTransformForItem(Child=(Type=Bone,Name="pelvis"), Parent=(Type=Bone,Name="root"), bChildInitial=False, bParentInitial=False)
   n32 = &RigVMFunction_MathDoubleRemap(SourceMaximum=20, SourceMinimum=30, TargetMaximum=1, TargetMinimum=0, Value=@n31.RelativeTransform.Translation.Z)
   > &RigUnit_ModifyTransforms(
    ItemToModify="((Item=(Type=Bone,Name=\"Jacket_base_l\"),Transform=(Rotation=(X=0,Y=-0.3841,Z=0,W=0.9233),Translation=(X=0,Y=0,Z=0),Scale3D=(X=1,Y=1,Z=1))),(Item=(Type=Bone,Name=\"Belt_front_r_strap\"),Transform=(Rotation=(X=0.444,Y=0.4364,Z=-0.4275,W=0.6555),Translation=(X= ...(truncated)"
    WeightMinimum=0
    Weight=&RigVMFunction_MathDoubleMul(B=@n32.Result, A=@n30.OutputParam).Result
   )
   > &RigUnit_ModifyTransforms(
    ItemToModify=((Item=(Type=Bone,Name="Belt_front_r_strap"),Transform=(Rotation=(X=0,Y=-0.342,Z=0,W=0.9397),Translation=(X=0,Y=2,Z=0),Scale3D=(X=1,Y=1,Z=1))))
    WeightMinimum=0
    Weight=&RigVMFunction_MathDoubleRemap(SourceMaximum=11, SourceMinimum=22, TargetMaximum=1, TargetMinimum=0, Value=@n31.RelativeTransform.Translation.Z).Result
   )
   > &RigUnit_ModifyTransforms(
    ItemToModify="((Item=(Type=Bone,Name=\"Jacketback_l_base\"),Transform=(Rotation=(X=0,Y=0.1564,Z=0,W=0.9877),Translation=(X=0,Y=0,Z=0),Scale3D=(X=1,Y=1,Z=1))),(Item=(Type=Bone,Name=\"Belt_front_r_strap\"),Transform=(Rotation=(X=0,Y=0,Z=0,W=1),Translation=(X=0,Y=0,Z=0),Scale3 ...(truncated)"
    WeightMinimum=0
    Weight=&RigVMFunction_MathDoubleRemap(
     SourceMaximum=50
     SourceMinimum=80
     TargetMaximum=1
     TargetMinimum=0
     Value=&RigUnit_GetRelativeTransformForItem(Child=(Type=Bone,Name="pelvis"), Parent=(Type=Bone,Name="root"), bChildInitial=False, bParentInitial=False).RelativeTransform.Translation.Z
    ).Result
   )
   > &RigUnit_ModifyTransforms(
    ItemToModify="((Item=(Type=Bone,Name=\"Belt_front_r_strap\"),Transform=(Rotation=(X=0,Y=0,Z=0,W=1),Translation=(X=0,Y=-3,Z=0),Scale3D=(X=1,Y=1,Z=1))),(Item=(Type=Bone,Name=\"Belt_front_l_strap\"),Transform=(Rotation=(X=0,Y=0,Z=-0.3969,W=0.9179),Translation=(X=0,Y=0,Z=-2.246 ...(truncated)"
    WeightMinimum=0
    Weight=@n32.Result
   )
   RigVMFunction_MathQuaternionToEuler = &RigVMFunction_MathQuaternionToEuler(
    RotationOrder=ZYX
    Value=&RigUnit_GetRelativeTransformForItem(Child=(Type=Bone,Name="foot_r"), Parent=(Type=Bone,Name="calf_r"), bChildInitial=False, bParentInitial=False).RelativeTransform.Rotation
   )
   &RigUnit_ModifyTransforms(
    ItemToModify=((Item=(Type=Bone,Name="Base_sneaker_tongue_r"),Transform=(Rotation=(X=0,Y=0.5551,Z=0,W=0.8318),Translation=(X=-1.7709,Y=0,Z=-4.7357),Scale3D=(X=1,Y=1,Z=1))))
    WeightMinimum=0
    Weight=&RigVMFunction_MathDoubleRemap(SourceMaximum=-60, SourceMinimum=-20, TargetMaximum=1, TargetMinimum=0, Value=@RigVMFunction_MathQuaternionToEuler.Result.Z).Result
    ExecutePin=@RigVMFunction_Sequence_1.F
   )
   n33 = &RigUnit_GetRelativeTransformForItem(Child=(Type=Bone,Name="foot_l"), Parent=(Type=Bone,Name="calf_l"), bChildInitial=False, bParentInitial=False)
   > &RigUnit_ModifyTransforms(
    ItemToModify=((Item=(Type=Bone,Name="Base_sneaker_tongue_l"),Transform=(Rotation=(X=0,Y=0.5405,Z=0,W=0.8413),Translation=(X=-1.7709,Y=0,Z=-4.7357),Scale3D=(X=1,Y=1,Z=1))))
    WeightMinimum=0
    Weight=&RigVMFunction_MathDoubleRemap(
     SourceMaximum=-60
     SourceMinimum=-20
     TargetMaximum=1
     TargetMinimum=0
     Value=&RigVMFunction_MathQuaternionToEuler(RotationOrder=ZYX, Value=@n33.RelativeTransform.Rotation).Result.Z
    ).Result
   )
   &RigUnit_ModifyTransforms(
    ItemToModify=((Item=(Type=Bone,Name="Jacket_base_r"),Transform=(Rotation=(X=0,Y=0,Z=0,W=1),Translation=(X=0,Y=0,Z=0),Scale3D=(X=1,Y=1,Z=1))),(Item=(Type=Bone,Name="Jacket_base_l"),Transform=(Rotation=(X=0,Y=0,Z=0,W=1),Translation=(X=0,Y=0,Z=0),Scale3D=(X=1,Y=1,Z=1))))
    WeightMinimum=0
    Weight=&RigVMFunction_MathDoubleRemap(
     SourceMaximum=20
     SourceMinimum=30
     TargetMaximum=1
     TargetMinimum=0
     Value=&RigUnit_GetRelativeTransformForItem(Child=(Type=Bone,Name="pelvis"), Parent=(Type=Bone,Name="root"), bChildInitial=False, bParentInitial=False).RelativeTransform.Translation.Z
    ).Result
    ExecutePin=@RigVMFunction_Sequence_1.H
   )
   Multiply = &RigVMFunction_MathDoubleMul(
    A=&RigVMFunction_MathDoubleRemap(
     SourceMaximum=70
     SourceMinimum=50
     TargetMaximum=1
     TargetMinimum=0
     Value=&RigUnit_GetRelativeTransformForItem(Child=(Type=Bone,Name="pelvis"), Parent=(Type=Bone,Name="root"), bChildInitial=False, bParentInitial=False).RelativeTransform.Translation.Z
    ).Result
    B=&RigVMFunction_MathDoubleRemap(
     SourceMaximum=65
     SourceMinimum=80
     TargetMaximum=1
     TargetMinimum=0
     Value=&RigUnit_GetRelativeTransformForItem(Child=(Type=Bone,Name="pelvis"), Parent=(Type=Bone,Name="root"), bChildInitial=False, bParentInitial=False).RelativeTransform.Translation.Z
    ).Result
   )
   > &RigUnit_ModifyTransforms(
    ItemToModify="((Item=(Type=Bone,Name=\"Jacket_base_r\"),Transform=(Rotation=(X=0,Y=0,Z=0.2588,W=0.9659),Translation=(X=0,Y=-2,Z=0),Scale3D=(X=1,Y=1,Z=1))),(Item=(Type=Bone,Name=\"Jacket_base_l\"),Transform=(Rotation=(X=0,Y=0,Z=0,W=1),Translation=(X=0,Y=0,Z=0),Scale3D=(X=1,Y ...(truncated)"
    WeightMinimum=0
    Weight=@Multiply.Result
   )
   n34 = &RigVMFunction_ControlFlowBranch(ExecuteContext=@RigVMFunction_Sequence_1.I, Condition=RigVMVariableNode() {variable=isLod2}.Value)
   &RigUnit_ModifyTransforms(
    ItemToModify="((Item=(Type=Bone,Name=\"corrective_alt_Back_r\"),Transform=(Rotation=(X=0,Y=0,Z=0,W=1),Translation=(X=0,Y=0,Z=1.7095),Scale3D=(X=1,Y=1,Z=1))),(Item=(Type=Bone,Name=\"corrective_alt_Back_l\"),Transform=(Rotation=(X=0,Y=0,Z=0,W=1),Translation=(X=0,Y=0,Z=4),Scal ...(truncated)"
    Weight=1
    WeightMinimum=0
    ExecutePin=@n34.False
   )
   &RigUnit_ModifyTransforms(
    ItemToModify="((Item=(Type=Bone,Name=\"corrective_alt_Back_r\"),Transform=(Rotation=(X=0,Y=0,Z=0,W=1),Translation=(X=0,Y=0,Z=1.7095),Scale3D=(X=1,Y=1,Z=1))),(Item=(Type=Bone,Name=\"corrective_alt_Back_l\"),Transform=(Rotation=(X=0,Y=0,Z=0,W=1),Translation=(X=0,Y=0,Z=4),Scal ...(truncated)"
    Weight=1
    WeightMinimum=0
    ExecutePin=@n34.True
   )
   n35 = &RigUnit_SphericalPoseReader(
    ActiveRegionScaleFactors=(PositiveWidth=1,NegativeWidth=0,PositiveHeight=1,NegativeHeight=1)
    ActiveRegionSize=0.3
    Debug=(bDrawDebug=false,bDraw2D=false,bDrawLocalAxes=false,DebugScale=30,DebugSegments=30,DebugThickness=0.2)
    DriverAxis=(X=1,Y=0,Z=0)
    DriverItem=(Type=Bone,Name="thigh_r")
    ExecutePin=()
    FalloffRegionScaleFactors=(PositiveWidth=1,NegativeWidth=1,PositiveHeight=1,NegativeHeight=1)
    FalloffSize=0.5
    OptionalParentItem=(Type=Bone,Name="pelvis")
    RotationOffset=(X=-90,Y=180,Z=-26)
   )
   > &RigUnit_ModifyTransforms(
    ItemToModify=((Item=(Type=Bone,Name="Belt_front_r_strap"),Transform=(Rotation=(X=-0.4101,Y=0.602,Z=-0.6664,W=0.1593),Translation=(X=-5.144,Y=-1.3,Z=-8.5),Scale3D=(X=1,Y=1,Z=1))))
    WeightMinimum=0
    Weight=@n35.OutputParam
   )
   n36 = &RigUnit_GetRelativeTransformForItem(Child=(Type=Bone,Name="pelvis"), Parent=(Type=Bone,Name="root"), bChildInitial=False, bParentInitial=False)
   n37 = &RigUnit_GetRelativeTransformForItem(Child=(Type=Bone,Name="pelvis"), Parent=(Type=Bone,Name="root"), bChildInitial=False, bParentInitial=False)
   &RigUnit_ModifyTransforms(
    ExecutePin=()
    ItemToModify="((Item=(Type=Bone,Name=\"Jacketback_base_backside\"),Transform=(Rotation=(X=0,Y=0,Z=0,W=1),Translation=(X=5,Y=0,Z=0),Scale3D=(X=1,Y=1,Z=1))),(Item=(Type=Bone,Name=\"Jacket_base_l\"),Transform=(Rotation=(X=0.3368,Y=0.1632,Z=-0.0594,W=0.9254),Translation=(X=0,Y= ...(truncated)"
    WeightMinimum=0
    Weight=&RigVMFunction_MathDoubleMul(
     A=&RigVMFunction_MathDoubleRemap(SourceMaximum=40, SourceMinimum=22, TargetMaximum=1, TargetMinimum=0, Value=@n36.RelativeTransform.Translation.Z).Result
     B=&RigVMFunction_MathDoubleRemap(SourceMaximum=35, SourceMinimum=50, TargetMaximum=1, TargetMinimum=0, Value=@n37.RelativeTransform.Translation.Z).Result
    ).Result
   )
   &RigVMFunction_MathDoubleMul(A=1, B=@n18.Result)
   &RigUnit_ParentConstraint(
    AdvancedSettings=(InterpolationType=Average,RotationOrderForFilter=XZY)
    Child=(Type=Bone,Name="Belt_front_r_strap")
    ExecutePin=()
    Parents=((Item=(Type=Bone,Name="thigh_r"),Weight=0.3))
    Weight=1
   )
  ]}
]
hierarchy [
 root {loc=<0 0 0>, rot=<0 0 0 1>}
 pelvis {parent=root, loc=<0 -0.7071 93.1191>, rot=<0.0127 0.707 -0.0127 -0.707>}
 spine_01 {parent=pelvis, loc=<6.7494 -0.6388 0>, rot=<0 0 0.034 -0.9994>}
 spine_02 {parent=spine_01, loc=<7.2518 -0.035 0>, rot=<0 0 0.0238 -0.9997>}
 spine_03 {parent=spine_02, loc=<6.0104 0 0>, rot=<0 0 0.0149 -0.9999>}
 spine_04 {parent=spine_03, loc=<8.1024 0 0>, rot=<0 0 -0.0132 -0.9999>}
 spine_05 {parent=spine_04, loc=<10.0146 0 0>, rot=<0 0 -0.0015 -1>}
 attach_cape {parent=spine_05, loc=<9.0118 -11.7134 -0.0005>, rot=<0.0616 -0.7044 -0.0616 -0.7044>}
 neck_01 {parent=spine_05, loc=<11.3138 0.1529 0>, rot=<0 0 -0.2003 -0.9797>}
 neck_02 {parent=neck_01, loc=<4.5627 0 0>, rot=<0 0 0.0504 -0.9987>}
 head {parent=neck_02, loc=<5.3903 -0.0002 0>, rot=<0 0 0.115 -0.9934>}
 attach_backpack {parent=spine_05, loc=<-0.1259 -13.4724 -0.0005>, rot=<-0.0093 -0.707 0.0093 -0.707>}
 clavicle_l {parent=spine_05, loc=<4.3575 -0.1466 -2.3223>, rot=<0.0465 -0.7289 -0.009 -0.6829>}
 upperarm_l {parent=clavicle_l, loc=<15.5111 0 0>, rot=<0.0226 -0.3234 -0.0088 -0.9459>}
 lowerarm_l {parent=upperarm_l, loc=<26.6514 0 0>, rot=<0.0391 -0.0099 -0.2453 -0.9686>}
 hand_l {parent=lowerarm_l, loc=<22.3491 0 0>, rot=<-0.7081 -0.0507 0.0975 -0.6975>}
 thumb_01_l {parent=hand_l, loc=<1.3118 -0.3191 -2.459>, rot=<0.6546 -0.4042 0.0664 -0.6354>}
 thumb_02_l {parent=thumb_01_l, loc=<5.9336 -0.0448 0.0401>, rot=<-0.0037 0.0213 0.298 -0.9543>}
 thumb_03_l {parent=thumb_02_l, loc=<2.2223 0.3682 -0.0907>, rot=<0.0118 -0.013 0.0066 -0.9998>}
 index_metacarpal_l {parent=hand_l, loc=<1.6535 0.0183 -2.96>, rot=<0.0158 -0.0124 -0.0173 -0.9996>}
 index_01_l {parent=index_metacarpal_l, loc=<8.3008 0.0357 -0.0524>, rot=<-0.0044 0.0485 0.0924 -0.9945>}
 index_02_l {parent=index_01_l, loc=<3.3587 -0.3004 0.0452>, rot=<-0.002 0.0027 0.3225 -0.9466>}
 index_03_l {parent=index_02_l, loc=<2.5578 -0.123 -0.0252>, rot=<-0.002 0.0067 0.2558 -0.9667>}
 middle_metacarpal_l {parent=hand_l, loc=<2.1856 -0.1725 -0.5984>, rot=<0.006 0.0168 -0.0329 -0.9993>}
 middle_01_l {parent=middle_metacarpal_l, loc=<7.3275 -0.2447 0.0129>, rot=<-0.0208 0.0042 0.0858 -0.9961>}
 middle_02_l {parent=middle_01_l, loc=<4.0123 0.1401 0.1175>, rot=<-0.0218 0.0258 0.3512 -0.9357>}
 middle_03_l {parent=middle_02_l, loc=<2.6077 0.0534 0.1373>, rot=<0.0082 0.0188 0.1708 -0.9851>}
 ring_metacarpal_l {parent=hand_l, loc=<2.2074 -0.3625 1.3638>, rot=<-0.0284 0.0722 -0.0458 -0.9959>}
 ring_01_l {parent=ring_metacarpal_l, loc=<7.0295 -0.6197 -0.1446>, rot=<-0.0489 -0.0252 0.1892 -0.9804>}
 ring_02_l {parent=ring_01_l, loc=<3.6174 0.4647 -0.1027>, rot=<0 0 0.2806 -0.9598>}
 ring_03_l {parent=ring_02_l, loc=<2.4152 0.114 0.0199>, rot=<0 0 0.1189 -0.9929>}
 pinky_metacarpal_l {parent=hand_l, loc=<1.7616 -0.9753 3.3814>, rot=<-0.0397 0.0814 -0.069 -0.9935>}
 pinky_01_l {parent=pinky_metacarpal_l, loc=<7.1709 -0.4831 -0.1094>, rot=<-0.0654 -0.0035 0.2242 -0.9723>}
 pinky_02_l {parent=pinky_01_l, loc=<2.8274 0.0893 -0.1043>, rot=<0.0037 0.0011 0.2854 -0.9584>}
 pinky_03_l {parent=pinky_02_l, loc=<2.2046 0.0307 -0.0078>, rot=<-0.0002 0.0098 0.1445 -0.9895>}
 deform_wrist_dn_base_l {parent=hand_l, loc=<0 -2.2342 0>, rot=<0 0 0 -1>}
 deform_wrist_dn_l {parent=deform_wrist_dn_base_l, loc=<0 0 0>, rot=<0 0 0 1>}
 deform_wrist_up_base_l {parent=hand_l, loc=<0 2.2682 0>, rot=<0 0 0 -1>}
 deform_wrist_up_l {parent=deform_wrist_up_base_l, loc=<0 0 0>, rot=<0 0 0 1>}
 lowerarm_twist_02_l {parent=lowerarm_l, loc=<9.6199 0 0>, rot=<0 0 0 1>}
 lowerarm_twist_01_l {parent=lowerarm_l, loc=<17.8549 0 0>, rot=<0 0 0 1>}
 deform_elbow_in_base_l {parent=lowerarm_l, loc=<0.2112 5.259 0.8274>, rot=<0 0 0 1>}
 deform_elbow_in_l {parent=deform_elbow_in_base_l, loc=<0 0 0>, rot=<0 0 0 1>}
 deform_elbow_out_base_l {parent=lowerarm_l, loc=<0 -4 0>, rot=<0 0 0 1>}
 deform_elbow_out_l {parent=deform_elbow_out_base_l, loc=<0 0 0>, rot=<0 0 0 1>}
 upperarm_twist_01_l {parent=upperarm_l, loc=<0.0041 0 0>, rot=<0 0 0 1>}
 deform_upperarm_bk_base_l {parent=upperarm_twist_01_l, loc=<1.0813 -7.2193 -1.9453>, rot=<0 0 0 1>}
 deform_upperarm_bk_l {parent=deform_upperarm_bk_base_l, loc=<0 0 0>, rot=<0 0 0 1>}
 deform_upperarm_fr_base_l {parent=upperarm_twist_01_l, loc=<-0.3691 7.0506 -2.765>, rot=<0 0 0 1>}
 deform_upperarm_fr_l {parent=deform_upperarm_fr_base_l, loc=<0 0 0>, rot=<0 0 0 1>}
 deform_upperarm_up_base_l {parent=upperarm_twist_01_l, loc=<-3.2216 -1.1391 3.3311>, rot=<0 0 0 1>}
 deform_upperarm_up_l {parent=deform_upperarm_up_base_l, loc=<0 0 0>, rot=<0 0 0 1>}
 upperarm_twist_02_l {parent=upperarm_l, loc=<13.7941 0 0>, rot=<0 0 0 1>}
 clavicle_r {parent=spine_05, loc=<4.3575 -0.1466 2.3223>, rot=<0.7289 0.0465 -0.6829 0.009>}
 upperarm_r {parent=clavicle_r, loc=<-15.5111 0 0>, rot=<0.0226 -0.3234 -0.0088 -0.9459>}
 lowerarm_r {parent=upperarm_r, loc=<-26.6514 0 0>, rot=<0.0391 -0.0099 -0.2453 -0.9686>}
 hand_r {parent=lowerarm_r, loc=<-22.3491 0 0>, rot=<-0.7081 -0.0507 0.0975 -0.6975>}
 thumb_01_r {parent=hand_r, loc=<-1.3118 0.3191 2.459>, rot=<0.6546 -0.4042 0.0664 -0.6354>}
 thumb_02_r {parent=thumb_01_r, loc=<-5.9336 0.0448 -0.0401>, rot=<-0.0037 0.0213 0.298 -0.9543>}
 thumb_03_r {parent=thumb_02_r, loc=<-2.2223 -0.3682 0.0907>, rot=<0.0118 -0.013 0.0066 -0.9998>}
 index_metacarpal_r {parent=hand_r, loc=<-1.6535 -0.0183 2.96>, rot=<0.0158 -0.0124 -0.0173 -0.9996>}
 index_01_r {parent=index_metacarpal_r, loc=<-8.3008 -0.0357 0.0524>, rot=<-0.0044 0.0485 0.0924 -0.9945>}
 index_02_r {parent=index_01_r, loc=<-3.3587 0.3004 -0.0452>, rot=<-0.002 0.0027 0.3225 -0.9466>}
 index_03_r {parent=index_02_r, loc=<-2.5578 0.123 0.0252>, rot=<-0.002 0.0067 0.2558 -0.9667>}
 middle_metacarpal_r {parent=hand_r, loc=<-2.1856 0.1725 0.5984>, rot=<0.006 0.0168 -0.0329 -0.9993>}
 middle_01_r {parent=middle_metacarpal_r, loc=<-7.3275 0.2447 -0.0129>, rot=<-0.0208 0.0042 0.0858 -0.9961>}
 middle_02_r {parent=middle_01_r, loc=<-4.0123 -0.1401 -0.1175>, rot=<-0.0218 0.0258 0.3512 -0.9357>}
 middle_03_r {parent=middle_02_r, loc=<-2.6077 -0.0534 -0.1373>, rot=<0.0082 0.0188 0.1708 -0.9851>}
 ring_metacarpal_r {parent=hand_r, loc=<-2.2074 0.3625 -1.3638>, rot=<-0.0284 0.0722 -0.0458 -0.9959>}
 ring_01_r {parent=ring_metacarpal_r, loc=<-7.0295 0.6197 0.1446>, rot=<-0.0489 -0.0252 0.1892 -0.9804>}
 ring_02_r {parent=ring_01_r, loc=<-3.6174 -0.4647 0.1027>, rot=<0 0 0.2806 -0.9598>}
 ring_03_r {parent=ring_02_r, loc=<-2.4152 -0.114 -0.0199>, rot=<0 0 0.1189 -0.9929>}
 pinky_metacarpal_r {parent=hand_r, loc=<-1.7616 0.9753 -3.3814>, rot=<-0.0397 0.0814 -0.069 -0.9935>}
 pinky_01_r {parent=pinky_metacarpal_r, loc=<-7.1709 0.4831 0.1094>, rot=<-0.0654 -0.0035 0.2242 -0.9723>}
 pinky_02_r {parent=pinky_01_r, loc=<-2.8274 -0.0893 0.1043>, rot=<0.0037 0.0011 0.2854 -0.9584>}
 pinky_03_r {parent=pinky_02_r, loc=<-2.2046 -0.0307 0.0078>, rot=<-0.0002 0.0098 0.1445 -0.9895>}
 deform_wrist_dn_base_r {parent=hand_r, loc=<-0.0004 2.2345 -0.0001>, rot=<0 0 0 -1>}
 deform_wrist_dn_r {parent=deform_wrist_dn_base_r, loc=<0 0 0>, rot=<0 0 0 1>}
 deform_wrist_up_base_r {parent=hand_r, loc=<0.0003 -2.2684 0.0001>, rot=<0 0 0 -1>}
 deform_wrist_up_r {parent=deform_wrist_up_base_r, loc=<0 0 0>, rot=<0 0 0 1>}
 clock {parent=hand_r, loc=<1.1901 -2.7121 -0.6486>, rot=<-0.7497 -0.0657 -0.0193 -0.6582>}
 lowerarm_twist_02_r {parent=lowerarm_r, loc=<-9.6199 0 0>, rot=<0 0 0 1>}
 lowerarm_twist_01_r {parent=lowerarm_r, loc=<-17.8549 0 0>, rot=<0 0 0 1>}
 deform_elbow_in_base_r {parent=lowerarm_r, loc=<-0.2111 -5.2591 -0.8275>, rot=<0 0 0 -1>}
 deform_elbow_in_r {parent=deform_elbow_in_base_r, loc=<-0.0001 0 0.0002>, rot=<0 0 0 1>}
 deform_elbow_out_base_r {parent=lowerarm_r, loc=<0 4 -0.0001>, rot=<0 0 0 -1>}
 deform_elbow_out_r {parent=deform_elbow_out_base_r, loc=<0 0 0>, rot=<0 0 0 1>}
 upperarm_twist_01_r {parent=upperarm_r, loc=<-0.0041 0 0>, rot=<0 0 0 1>}
 deform_upperarm_bk_base_r {parent=upperarm_twist_01_r, loc=<-1.0814 7.2192 1.9455>, rot=<0 0 0 -1>}
 deform_upperarm_bk_r {parent=deform_upperarm_bk_base_r, loc=<0 0 0>, rot=<0 0 0 1>}
 deform_upperarm_fr_base_r {parent=upperarm_twist_01_r, loc=<0.3689 -7.0506 2.7652>, rot=<0 0 0 -1>}
 deform_upperarm_fr_r {parent=deform_upperarm_fr_base_r, loc=<0 0 0>, rot=<0 0 0 1>}
 deform_upperarm_up_base_r {parent=upperarm_twist_01_r, loc=<3.2214 1.1391 -3.3308>, rot=<0 0 0 -1>}
 deform_upperarm_up_r {parent=deform_upperarm_up_base_r, loc=<0 0 0>, rot=<0 0 0 -1>}
 upperarm_twist_02_r {parent=upperarm_r, loc=<-13.7941 0 0>, rot=<0 0 0 1>}
 Base_jaket_collar_l {parent=spine_05, loc=<3.8605 10.1024 4.2399>, rot=<-0.0783 0.152 -0.4131 -0.8945>}
 dyn_collar_l {parent=Base_jaket_collar_l, loc=<0 0 0>, rot=<0 0 0 1>}
 Base_jaket_collar_r {parent=spine_05, loc=<3.8605 10.1024 -4.4394>, rot=<0.9088 0.3637 -0.1347 0.1539>}
 dyn_collar_r {parent=Base_jaket_collar_r, loc=<0 0 0>, rot=<0 0 0 1>}
 Collar_base_midle_ctrl {parent=spine_05, loc=<15.9772 -6.0286 0>, rot=<0.0284 -0.7065 -0.0284 -0.7065>}
 Collar_ctrl_m {parent=Collar_base_midle_ctrl, loc=<0 0 0>, rot=<0 0 0 1>}
 Collar_base_right_ctrl {parent=spine_05, loc=<15.5768 -4.0842 7.0105>, rot=<0.2392 -0.6654 0.1851 -0.6825>}
 Collar_ctrl_r {parent=Collar_base_right_ctrl, loc=<0 0 0>, rot=<0 0 0 1>}
 Collar_base_left_ctrl {parent=spine_05, loc=<15.5768 -4.185 -7.0818>, rot=<-0.1369 -0.6937 -0.1921 -0.6805>}
 Collar_ctrl_l {parent=Collar_base_left_ctrl, loc=<0 0 0>, rot=<0 0 0 1>}
 Collar_base_black_ctrl {parent=spine_05, loc=<15.8171 -4.0378 0>, rot=<0.0284 -0.7065 -0.0284 -0.7065>}
 Collar_ctrl_black {parent=Collar_base_black_ctrl, loc=<0 0 0>, rot=<0 0 0 1>}
 corrective_alt_Front_r {parent=spine_05, loc=<-1.3145 9.3955 12.5052>, rot=<0.9103 -0.09 -0.0131 0.4039>}
 corrective_alt_Front_l {parent=spine_05, loc=<-1.0939 9.2237 -12.8077>, rot=<-0.4864 0.0642 0.0645 -0.869>}
 Collar_l_base {parent=spine_05, loc=<14.126 1.068 -9.185>, rot=<0.6742 0.1072 0.6939 -0.2288>}
 Collar_l {parent=Collar_l_base, loc=<0 0 0>, rot=<0 0 0 1>}
 Collar_r_base {parent=spine_05, loc=<14.1262 1.0682 9.1846>, rot=<0.2288 -0.6939 -0.1072 -0.6742>}
 Collar_r {parent=Collar_r_base, loc=<0 0 0>, rot=<0 0 0 1>}
 Collar_black_l_base {parent=spine_05, loc=<13.7502 2.3402 -7.8767>, rot=<0.6742 0.1072 0.6939 -0.2288>}
 Collar_black_l {parent=Collar_black_l_base, loc=<0 0 0>, rot=<0 0 0 1>}
 Collar_black_r_base {parent=spine_05, loc=<13.7551 2.2821 7.877>, rot=<0.2288 -0.6939 -0.1072 -0.6742>}
 Collar_black_r {parent=Collar_black_r_base, loc=<0 0 0>, rot=<0 0 0 1>}
 Jacket_base_l {parent=spine_03, loc=<0.8069 6.6241 -10.3525>, rot=<-0.0781 -0.3199 0.9351 -0.1308>}
 dyn_Jacket_l_01 {parent=Jacket_base_l, loc=<0 0 0>, rot=<-0.0696 0.0021 0.0298 -0.9971>}
 dyn_Jacket_l_02 {parent=dyn_Jacket_l_01, loc=<9.6736 0 0>, rot=<0.0034 -0.0516 -0.0663 -0.9965>}
 dyn_Jacket_l_03 {parent=dyn_Jacket_l_02, loc=<8 0 0>, rot=<0 0 0 1>}
 dyn_Jacket_l_05 {parent=dyn_Jacket_l_03, loc=<0 0 0>, rot=<0.0281 0.0435 0.5227 -0.851>}
 dyn_Jacket_l_06 {parent=dyn_Jacket_l_05, loc=<5.765 0 0>, rot=<0 0 0 1>}
 Jacket_base_r {parent=spine_03, loc=<0.807 6.624 10.353>, rot=<0.0881 -0.9324 0.3442 0.0667>}
 dyn_Jacket_r_01 {parent=Jacket_base_r, loc=<0 0 0>, rot=<0.0632 -0.0019 0.0298 -0.9976>}
 dyn_Jacket_r_02 {parent=dyn_Jacket_r_01, loc=<9.6732 0 0>, rot=<0.0614 -0.0332 0.0718 -0.995>}
 dyn_Jacket_r_03 {parent=dyn_Jacket_r_02, loc=<8 0 0>, rot=<0 0 0 1>}
 dyn_Jacket_r_05 {parent=dyn_Jacket_r_03, loc=<0 0 0>, rot=<-0.0477 0.0649 -0.5899 -0.8034>}
 dyn_Jacket_r_06 {parent=dyn_Jacket_r_05, loc=<5.765 0 0>, rot=<0 0 0 1>}
 Jacketback_r_base {parent=spine_03, loc=<3.1383 -3.5714 13.42>, rot=<0.0767 -0.9627 -0.259 0.0147>}
 dyn_Jacketback_r_01 {parent=Jacketback_r_base, loc=<0 0 0>, rot=<0 0 0 1>}
 dyn_Jacketback_r_02 {parent=dyn_Jacketback_r_01, loc=<7.8789 0 0>, rot=<0 0 0 1>}
 Jacketback_l_base {parent=spine_03, loc=<3.1383 -3.5714 -13.4199>, rot=<-0.0256 0.257 0.9632 -0.0738>}
 dyn_Jacketback_l_01 {parent=Jacketback_l_base, loc=<0 0 0>, rot=<0 0 0 1>}
 dyn_Jacketback_l_02 {parent=dyn_Jacketback_l_01, loc=<7.8789 0 0>, rot=<0 0 0 1>}
 Jacketback_base_backside {parent=spine_03, loc=<4.2173 -9.4484 -0.0111>, rot=<0.0508 -0.7177 -0.6933 0.0405>}
 dyn_Jacketback_backside_01 {parent=Jacketback_base_backside, loc=<0 0 0>, rot=<0 0 0 1>}
 dyn_Jacketback_backside_02 {parent=dyn_Jacketback_backside_01, loc=<7.8789 0 0>, rot=<0 0 0 1>}
 corrective_alt_Back_r {parent=spine_03, loc=<3.0868 -8.8566 9.2668>, rot=<0.8699 -0.0184 -0.0669 -0.4883>}
 corrective_alt_Back_l {parent=spine_03, loc=<2.5345 -7.0786 -11.1476>, rot=<0.4116 -0.0302 0.042 -0.9099>}
 thigh_l {parent=pelvis, loc=<-2.5693 -0.0109 -7.7501>, rot=<0.0012 0.0572 0.0155 -0.9982>}
 calf_l {parent=thigh_l, loc=<-39.0484 0 0>, rot=<0.0006 -0.0161 -0.0391 -0.9991>}
 foot_l {parent=calf_l, loc=<-42.561 0.1077 0.4292>, rot=<0.0418 -0.0455 0.0343 -0.9975>}
 ball_l {parent=foot_l, loc=<-6.1599 13.8264 0.6806>, rot=<0.0001 -0.0001 -0.7063 -0.7079>}
 Base_shoes_l {parent=foot_l, loc=<3.7549 -4.795 -0.042>, rot=<0.0018 0.6743 0.7384 0.0055>}
 dyn_shoes_l {parent=Base_shoes_l, loc=<-0.124 0.3971 -0.6081>, rot=<0 0 0 1>}
 Base_sneaker_tongue_l {parent=foot_l, loc=<3.4218 4.8313 0.3641>, rot=<-0.7384 0.0055 0.0018 -0.6743>}
 sneaker_tongue_l {parent=Base_sneaker_tongue_l, loc=<0 0 0>, rot=<0 0 0 1>}
 calf_twist_02_l {parent=calf_l, loc=<-18.8064 -0.0967 0.0397>, rot=<0.0237 0.0095 -0.0035 -0.9997>}
 calf_twist_01_l {parent=calf_l, loc=<-35.813 -0.0967 0.0397>, rot=<0.0237 0.0095 -0.0035 -0.9997>}
 deform_knee_in_base_l {parent=calf_l, loc=<1.298 -5.2957 -0.0496>, rot=<0 0 0 -1>}
 deform_knee_in_l {parent=deform_knee_in_base_l, loc=<0 0 0>, rot=<0 0 0 1>}
 kneecap {parent=calf_l, loc=<0 0 0>, rot=<-0.0315 -0.7349 0.027 -0.6769>}
 thigh_twist_01_l {parent=thigh_l, loc=<-21.6303 0 0>, rot=<0.0013 0 0 -1>}
 deform_groin_base_l {parent=thigh_l, loc=<-2.1908 8.1142 -1.7442>, rot=<0 0 0 -1>}
 deform_groin_l {parent=deform_groin_base_l, loc=<0 0 0>, rot=<0 0 0 1>}
 thigh_r {parent=pelvis, loc=<-2.5693 -0.0109 7.7501>, rot=<0.0572 -0.0012 0.9982 0.0155>}
 calf_r {parent=thigh_r, loc=<39.0484 0 0>, rot=<0.0006 -0.0161 -0.0391 -0.9991>}
 foot_r {parent=calf_r, loc=<42.561 -0.1077 -0.4292>, rot=<0.0418 -0.0455 0.0343 -0.9975>}
 ball_r {parent=foot_r, loc=<6.1599 -13.8264 -0.6806>, rot=<0.0001 -0.0001 -0.7063 -0.7079>}
 Base_shoes_r {parent=foot_r, loc=<-3.7549 4.795 0.042>, rot=<-0.7384 0.0055 0.0018 -0.6743>}
 dyn_shoes_r {parent=Base_shoes_r, loc=<-0.1241 -0.3971 -0.608>, rot=<0 0 0 1>}
 Base_sneaker_tongue_r {parent=foot_r, loc=<-3.4218 -4.8313 -0.3641>, rot=<0.0018 0.6743 0.7384 0.0055>}
 sneaker_tongue_r {parent=Base_sneaker_tongue_r, loc=<0 0 0>, rot=<0 0 0 1>}
 calf_twist_02_r {parent=calf_r, loc=<18.8064 0.0967 -0.0397>, rot=<0.0237 0.0095 -0.0035 -0.9997>}
 calf_twist_01_r {parent=calf_r, loc=<35.813 0.0967 -0.0397>, rot=<0.0237 0.0095 -0.0035 -0.9997>}
 deform_knee_in_base_r {parent=calf_r, loc=<-1.298 5.2957 0.0496>, rot=<0 0 0 1>}
 deform_knee_in_r {parent=deform_knee_in_base_r, loc=<0 0 0>, rot=<0 0 0 1>}
 deform_knee_out_base_r {parent=calf_r, loc=<0 -4 0>, rot=<0 0 0 -1>}
 deform_knee_out_r {parent=deform_knee_out_base_r, loc=<0 0 0>, rot=<0 0 0 1>}
 thigh_twist_01_r {parent=thigh_r, loc=<21.6303 0 0>, rot=<0.0013 0 0 -1>}
 deform_groin_base_r {parent=thigh_r, loc=<2.1908 -8.1142 1.7442>, rot=<0 0 0 -1>}
 deform_groin_r {parent=deform_groin_base_r, loc=<0 0 0>, rot=<0 0 0 1>}
 bow {parent=pelvis, loc=<2.7129 10.1786 0.1669>, rot=<-0.0127 -0.707 0.0127 -0.707>}
 Belt_front_r_strap_master {parent=pelvis, loc=<-1.1131 7.9251 6.8116>, rot=<0.0104 -0.7473 0.6195 0.2401>}
 Belt_front_r_strap {parent=Belt_front_r_strap_master, loc=<0 0 0>, rot=<0 0 0 1>}
 dyn_Belt_01_front_r {parent=Belt_front_r_strap, loc=<3.7105 0 0>, rot=<0 0.0722 0 -0.9974>}
 dyn_Belt_02_front_r {parent=dyn_Belt_01_front_r, loc=<3.1164 0 0>, rot=<-0.0018 0.0684 0.1253 -0.9898>}
 dyn_Belt_04_front_r {parent=dyn_Belt_02_front_r, loc=<7.7649 0 0>, rot=<0.0887 0.0797 0.0144 -0.9928>}
 dyn_Belt_05_front_r {parent=dyn_Belt_04_front_r, loc=<5.4084 0 0>, rot=<0.0586 0.0805 0.0889 -0.9911>}
 dyn_Belt_06_front_r {parent=dyn_Belt_05_front_r, loc=<5.5718 0 0>, rot=<0.0998 0.0554 -0.0759 -0.9906>}
 dyn_Belt_07_front_r {parent=dyn_Belt_06_front_r, loc=<4.6698 -2.4563 2.4108>, rot=<0.036 0.3567 0.388 -0.8491>}
 Belt_front_l_strap_master {parent=pelvis, loc=<-2.1279 8.4143 -6.7389>, rot=<-0.289 -0.596 0.7438 -0.0893>}
 Belt_front_l_strap {parent=Belt_front_l_strap_master, loc=<0 0 0>, rot=<0 0 0 1>}
 dyn_Belt_01_front_l {parent=Belt_front_l_strap, loc=<2.4642 0 0>, rot=<0.0162 0.0539 0.0768 -0.9955>}
 dyn_Belt_02_front_l {parent=dyn_Belt_01_front_l, loc=<3.4346 0 0>, rot=<-0.0145 0.0675 -0.0528 -0.9962>}
 dyn_Belt_03_front_l {parent=dyn_Belt_02_front_l, loc=<7.141 0 0>, rot=<-0.0603 0.0522 0.0023 -0.9968>}
 dyn_Belt_04_front_l {parent=dyn_Belt_03_front_l, loc=<6.647 0 0>, rot=<-0.0313 0.1084 -0.017 -0.9935>}
 dyn_Belt_05_front_l {parent=dyn_Belt_04_front_l, loc=<5.9259 0 0>, rot=<-0.081 0.0843 -0.0301 -0.9927>}
 dyn_Belt_06_front_l {parent=dyn_Belt_05_front_l, loc=<6.933 1.9011 3.7609>, rot=<-0.1959 0.4585 -0.3269 -0.8028>}
 Master_Belt_01_backside_l {parent=pelvis, loc=<-0.4955 -11.7364 -6.1246>, rot=<-0.1799 0.5887 0.7848 0.0722>}
 Base_Belt_01_backside_l {parent=Master_Belt_01_backside_l, loc=<0 0 0>, rot=<0 0 0 1>}
 dyn_Belt_01_backside_l {parent=Base_Belt_01_backside_l, loc=<0 0 0>, rot=<0 0 0 1>}
 dyn_Belt_02_backside_l {parent=dyn_Belt_01_backside_l, loc=<3.0111 0 0>, rot=<0.0057 0.1178 -0.0037 -0.993>}
 dyn_Belt_03_backside_l {parent=dyn_Belt_02_backside_l, loc=<8.3934 0 0>, rot=<0.0418 0.0568 0.0558 -0.9959>}
 dyn_Belt_04_backside_l {parent=dyn_Belt_03_backside_l, loc=<8.282 0 0>, rot=<0.0699 0.0385 0.023 -0.9965>}
 dyn_Belt_05_backside_l {parent=dyn_Belt_04_backside_l, loc=<7.3754 0 0>, rot=<0.0628 0.0102 0.0008 -0.998>}
 dyn_Belt_06_backside_l {parent=dyn_Belt_05_backside_l, loc=<6.6056 -3.6464 2.1019>, rot=<0.0602 0.2888 0.428 -0.8543>}
 Master_Belt_01_backside_r {parent=pelvis, loc=<-0.5589 -11.8879 6.147>, rot=<-0.0633 -0.7967 -0.585 0.1379>}
 Base_Belt_01_backside_r {parent=Master_Belt_01_backside_r, loc=<0 0 0>, rot=<0 0 0 1>}
 dyn_Belt_01_backside_r {parent=Base_Belt_01_backside_r, loc=<0 0 0>, rot=<0 0 0 1>}
 dyn_Belt_02_backside_r {parent=dyn_Belt_01_backside_r, loc=<3.0836 0 0>, rot=<0.0057 0.08 -0.0759 -0.9939>}
 dyn_Belt_04_backside_r {parent=dyn_Belt_02_backside_r, loc=<8.7151 0 0>, rot=<-0.0247 0.0695 -0.0843 -0.9937>}
 dyn_Belt_05_backside_r {parent=dyn_Belt_04_backside_r, loc=<8.324 0 0>, rot=<-0.0838 0.0506 -0.0895 -0.9912>}
 dyn_Belt_06_backside_r {parent=dyn_Belt_05_backside_r, loc=<7.0985 0 0>, rot=<-0.0748 0.0815 -0.0767 -0.9909>}
 dyn_Belt_07_backside_r {parent=dyn_Belt_06_backside_r, loc=<5.6196 1.7843 1.7254>, rot=<-0.0113 0.3615 -0.2404 -0.9008>}
]
variables [
 DBNO {type=bool, default=False}
 ItemToModify_0_Transform_Rotation {type=FQuat, default=(X=-0.1508,Y=0.2318,Z=-0.059,W=0.9592)}
 Sit {type=double, default=0}
 isCrouchingCR {type=double, default=0}
 isLod2 {type=bool, default=False}
 isMove {type=double, default=0}
 skydive {type=bool, default=False}
]
