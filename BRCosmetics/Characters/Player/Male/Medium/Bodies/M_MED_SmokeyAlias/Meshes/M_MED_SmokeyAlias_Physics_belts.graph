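# Physics asset for the SmokeyAlias male medium body's belts, letting four dangling strap chains (front and back, left and right) swing against the lower body.
# Limited-angle capsule chains hang from strap and base bodies on the pelvis, and each side's front and back ends are tightly constrained into a loop.
# Locked pelvis, thigh and calf bodies form collision proxies, with oversized thigh spheres and a huge root capsule set below the character.
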PhysicsAsset /BRCosmetics/Characters/Player/Male/Medium/Bodies/M_MED_SmokeyAlias/Meshes/M_MED_SmokeyAlias_Physics_belts
bodies {
 Base_Belt_01_backside_l [{shapeType=Capsule, length=0, radius=1}]
 Base_Belt_01_backside_r [{shapeType=Capsule, length=0, radius=1}]
 Belt_front_l_strap [{shapeType=Capsule, length=0, radius=0}]
 Belt_front_r_strap [{shapeType=Capsule, length=0, radius=0}]
 calf_l [{
   shapeType=Capsule
   length=40.397
   radius=6.3507
   shapeName=calf_l_capsule
   center=<-16.9336 -1.7156 0.2389>
   rotation=<rot -39.3966 -90.0001 89.8789>
  }]
 calf_r [{
   shapeType=Capsule
   length=38.2053
   radius=6.7014
   shapeName=calf_r_capsule
   center=<17.2803 0.8996 -0.0977>
   rotation=<rot 50.0272 -100.0903 89.0763>
  }]
 dyn_Belt_01_backside_l [{
   shapeType=Capsule
   length=2.7581
   radius=2.7265
   shapeName=dyn_Belt_01_backside_l_capsule
   center=<1.5 0 -1.5>
   rotation=<rot 0 1.0947 -89.2163>
  }]
 dyn_Belt_01_backside_r [{
   shapeType=Capsule
   length=3.3063
   radius=2.5377
   shapeName=dyn_Belt_01_backside_r_capsule
   center=<1.5 0 -1.5>
   rotation=<rot 0.0001 -2.4497 -87.7008>
  }]
 dyn_Belt_01_front_l [{shapeType=Capsule, length=3.9796, radius=2.5, center=<2 0 -2.2>, rotation=<rot -0.0001 0 -89.9036>}]
 dyn_Belt_01_front_r [{
   shapeType=Capsule
   length=3.2545
   radius=2.375
   shapeName=dyn_Belt_01_front_r_capsule
   center=<0.5993 0.0996 -2.0015>
   rotation=<rot 0.1048 -7.0879 -89.1215>
  }]
 dyn_Belt_02_backside_l [{
   shapeType=Capsule
   length=5.7316
   radius=2.4979
   shapeName=dyn_Belt_02_backside_l_capsule
   center=<3 0 -2>
   rotation=<rot 0.1835 -93.9899 -92.7292>
  }]
 dyn_Belt_02_backside_r [{
   shapeType=Capsule
   length=4.7334
   radius=2.4748
   shapeName=dyn_Belt_02_backside_r_capsule
   center=<4 0 -1.5>
   rotation=<rot 0.0406 -88.4908 -89.4434>
  }]
 dyn_Belt_02_front_l [{
   shapeType=Capsule
   length=4.1683
   radius=2.7491
   shapeName=dyn_Belt_02_front_l_capsule
   center=<3.9391 0 -2.3>
   rotation=<rot 1.2872 90.0753 -86.6451>
  }]
 dyn_Belt_02_front_r [{
   shapeType=Capsule
   length=6.035
   radius=2.0226
   shapeName=dyn_Belt_02_front_l_capsule
   center=<4.3933 0 -1>
   rotation=<rot 1.2345 90.9721 93.3746>
  }]
 dyn_Belt_03_backside_l [{
   shapeType=Capsule
   length=6.934
   radius=2.2979
   shapeName=dyn_Belt_03_backside_l_capsule
   center=<3.0018 0.0004 -1.4597>
   rotation=<rot 0.1122 -86.9403 -89.2709>
  }]
 dyn_Belt_03_front_l [{
   shapeType=Capsule
   length=7.2945
   radius=3.446
   shapeName=dyn_Belt_03_front_l_capsule
   center=<3 0 -2.8>
   rotation=<rot 0.1998 -90.6324 -89.7189>
  }]
 dyn_Belt_04_backside_l [{
   shapeType=Capsule
   length=6.9073
   radius=2.899
   shapeName=dyn_Belt_04_backside_l_capsule
   center=<3.0098 0.0205 -1.3803>
   rotation=<rot 3.7262 -91.9632 -88.5143>
  }]
 dyn_Belt_04_backside_r [{
   shapeType=Capsule
   length=9.4687
   radius=2.7933
   shapeName=dyn_Belt_04_backside_r_capsule
   center=<4 0 -1.7>
   rotation=<rot -0.0143 -89.7273 -89.9376>
  }]
 dyn_Belt_04_front_l [{
   shapeType=Capsule
   length=4.3916
   radius=2.4247
   shapeName=dyn_Belt_04_front_l_capsule
   center=<3.0018 0 -2>
   rotation=<rot 3.7627 92.315 -90.5665>
  }]
 dyn_Belt_04_front_r [{
   shapeType=Capsule
   length=6.6514
   radius=2.5305
   shapeName=dyn_Belt_04_front_r_capsule
   center=<2.4707 0 -1.3>
   rotation=<rot 0.3304 -90.1961 -91.8639>
  }]
 dyn_Belt_05_backside_l [{
   shapeType=Capsule
   length=6.5024
   radius=2.889
   shapeName=dyn_Belt_05_backside_l_capsule
   center=<3.8757 -1.9324 -0.8966>
   rotation=<rot 0 61.4626 -74.2952>
  }]
 dyn_Belt_05_backside_r [{
   shapeType=Capsule
   length=8.1245
   radius=2.4732
   shapeName=dyn_Belt_05_backside_r_capsule
   center=<3.5 0 -1.4>
   rotation=<rot -90 -0.0001 0.0003>
  }]
 dyn_Belt_05_front_l [{
   shapeType=Capsule
   length=8.1906
   radius=2.1128
   shapeName=dyn_Belt_05_front_l_capsule
   center=<3.3446 0 0>
   rotation=<rot -7.8245 103.2451 -70.1144>
  }]
 dyn_Belt_05_front_r [{
   shapeType=Capsule
   length=7.8783
   radius=2.7165
   shapeName=dyn_Belt_05_front_r_capsule
   center=<2.6214 0 -1.3>
   rotation=<rot -1.446 -87.1695 -93.1594>
  }]
 dyn_Belt_06_backside_l [{shapeType=Capsule, length=8.4121, radius=2.8656, center=<0 0 -1>, rotation=<rot -90 -179.9999 179.9998>}]
 dyn_Belt_06_backside_r [{
   shapeType=Capsule
   length=8.1013
   radius=2.6365
   shapeName=dyn_Belt_06_backside_r_capsule
   center=<2.2675 0.6365 -0.0033>
   rotation=<rot -0.2987 94.73 -82.8246>
  }]
 dyn_Belt_06_front_l [{shapeType=Capsule, length=8.4121, radius=2.8656, center=<0 0 -1>, rotation=<rot -90 -179.9802 179.9801>}]
 dyn_Belt_06_front_r [{
   shapeType=Capsule
   length=8.2832
   radius=3.0184
   shapeName=dyn_Belt_06_front_r_capsule
   center=<2.7071 -0.3131 -0.0712>
   rotation=<rot 11.1903 60.3874 -70.7851>
  }]
 dyn_Belt_07_backside_r [{shapeType=Capsule, length=7.4957, radius=3.2186, center=<0 0.2569 -1>, rotation=<rot -90 90 -89.9998>}]
 dyn_Belt_07_front_r [{shapeType=Capsule, length=7.4957, radius=3.2186, center=<0 -0.2569 -1>, rotation=<rot -90 0 -179.9997>}]
 pelvis [{
   shapeType=Capsule
   length=13.0319
   radius=10.2754
   shapeName=pelvis_capsule
   center=<-4.9539 -0.9956 -0.2442>
   rotation=<rot 2.204 -90 -1.1169>
  }]
 root [{shapeType=Capsule, length=18.5146, radius=139.5593, shapeName=root_capsule, center=<0 8.5053 -136.1863>, rotation=<rot -86.041 90 -180>}]
 thigh_l [
  {shapeType=Capsule, length=19.0491, radius=8.8822, center=<-16.0897 0.7035 0.6234>, rotation=<rot -87.1137 179.955 179.9999>}
  {shapeType=Capsule, length=0, radius=22.6693, shapeName=thigh_l_capsule, center=<-18.5434 0.0001 21.7752>, rotation=<rot 0 -0.0002 90>}
  {shapeType=Capsule, length=5.266, radius=4.87, shapeName=thigh_l_capsule, center=<-3 -5.986 0>, rotation=<rot 0 0 -12.1514>}
 ]
 thigh_r [
  {
   shapeType=Capsule
   length=21.4159
   radius=8.8303
   shapeName=thigh_r_capsule
   center=<17.4462 -0.8505 -1.2993>
   rotation=<rot -87.1137 179.9551 178.2324>
  }
  {shapeType=Capsule, length=0, radius=24.6333, shapeName=thigh_r_capsule, center=<21.509 4.7439 -24.8122>, rotation=<rot 0 -0.0004 75>}
  {
   shapeType=Capsule
   length=9.5217
   radius=5.5833
   shapeName=thigh_r_capsule
   center=<27.8577 -0.7058 2.7573>
   rotation=<rot 80.1602 -0.0001 0.0001>
  }
  {shapeType=Capsule, length=4.2248, radius=4.7013, shapeName=thigh_r_capsule, center=<3.4866 6.2466 -0.153>, rotation=<rot 0 0 -17.448>}
 ]
}
constraints [
 {
  bone1Name=calf_l
  bone2Name=thigh_l
  swing1LimitDegrees=45
  swing1Motion=Locked
  swing2LimitDegrees=45
  swing2Motion=Locked
  twistLimitDegrees=45
  twistMotion=Locked
 }
 {
  bone1Name=calf_r
  bone2Name=thigh_r
  swing1LimitDegrees=45
  swing1Motion=Locked
  swing2LimitDegrees=45
  swing2Motion=Locked
  twistLimitDegrees=45
  twistMotion=Locked
 }
 {
  bone1Name=thigh_l
  bone2Name=pelvis
  swing1LimitDegrees=45
  swing1Motion=Locked
  swing2LimitDegrees=45
  swing2Motion=Locked
  twistLimitDegrees=45
  twistMotion=Locked
 }
 {
  bone1Name=thigh_r
  bone2Name=pelvis
  swing1LimitDegrees=45
  swing1Motion=Locked
  swing2LimitDegrees=45
  swing2Motion=Locked
  twistLimitDegrees=45
  twistMotion=Locked
 }
 {
  bone1Name=pelvis
  bone2Name=root
  swing1LimitDegrees=45
  swing1Motion=Locked
  swing2LimitDegrees=45
  swing2Motion=Locked
  twistLimitDegrees=45
  twistMotion=Locked
 }
 {bone1Name=dyn_Belt_01_backside_l, bone2Name=Base_Belt_01_backside_l, swing1LimitDegrees=0, swing2LimitDegrees=20, twistLimitDegrees=2}
 {bone1Name=dyn_Belt_02_backside_l, bone2Name=dyn_Belt_01_backside_l, swing1LimitDegrees=1, swing2LimitDegrees=35, twistLimitDegrees=1}
 {bone1Name=dyn_Belt_03_backside_l, bone2Name=dyn_Belt_02_backside_l, swing1LimitDegrees=10, swing2LimitDegrees=10, twistLimitDegrees=2}
 {bone1Name=dyn_Belt_04_backside_l, bone2Name=dyn_Belt_03_backside_l, swing1LimitDegrees=10, swing2LimitDegrees=23, twistLimitDegrees=2}
 {bone1Name=dyn_Belt_05_backside_l, bone2Name=dyn_Belt_04_backside_l, swing1LimitDegrees=20, swing2LimitDegrees=20, twistLimitDegrees=2}
 {bone1Name=dyn_Belt_01_backside_r, bone2Name=Base_Belt_01_backside_r, swing1LimitDegrees=0, swing2LimitDegrees=20, twistLimitDegrees=2}
 {bone1Name=dyn_Belt_02_backside_r, bone2Name=dyn_Belt_01_backside_r, swing1LimitDegrees=1, swing2LimitDegrees=30, twistLimitDegrees=2}
 {bone1Name=dyn_Belt_04_backside_r, bone2Name=dyn_Belt_02_backside_r, swing1LimitDegrees=10, swing2LimitDegrees=10, twistLimitDegrees=2}
 {bone1Name=dyn_Belt_05_backside_r, bone2Name=dyn_Belt_04_backside_r, swing1LimitDegrees=15, swing2LimitDegrees=15, twistLimitDegrees=2}
 {bone1Name=dyn_Belt_06_backside_r, bone2Name=dyn_Belt_05_backside_r, swing1LimitDegrees=20, swing2LimitDegrees=20, twistLimitDegrees=2}
 {bone1Name=dyn_Belt_02_front_l, bone2Name=dyn_Belt_01_front_l, swing1LimitDegrees=8, swing2LimitDegrees=40, twistLimitDegrees=2}
 {bone1Name=dyn_Belt_03_front_l, bone2Name=dyn_Belt_02_front_l, swing1LimitDegrees=40, swing2LimitDegrees=45, twistLimitDegrees=15}
 {bone1Name=dyn_Belt_04_front_l, bone2Name=dyn_Belt_03_front_l, swing1LimitDegrees=20, swing2LimitDegrees=30, twistLimitDegrees=10}
 {bone1Name=dyn_Belt_05_front_l, bone2Name=dyn_Belt_04_front_l, swing1LimitDegrees=15, swing2LimitDegrees=20, twistLimitDegrees=5}
 {bone1Name=dyn_Belt_04_front_r, bone2Name=dyn_Belt_02_front_r, swing1LimitDegrees=20, swing2LimitDegrees=20, twistLimitDegrees=2}
 {bone1Name=dyn_Belt_05_front_r, bone2Name=dyn_Belt_04_front_r, swing1LimitDegrees=20, swing2LimitDegrees=20, twistLimitDegrees=2}
 {bone1Name=dyn_Belt_06_front_r, bone2Name=dyn_Belt_05_front_r, swing1LimitDegrees=25, swing2LimitDegrees=25, twistLimitDegrees=2}
 {bone1Name=dyn_Belt_01_front_l, bone2Name=Belt_front_l_strap, swing1LimitDegrees=5, swing2LimitDegrees=35, twistLimitDegrees=2}
 {
  bone1Name=Belt_front_l_strap
  bone2Name=pelvis
  swing1LimitDegrees=45
  swing1Motion=Locked
  swing2LimitDegrees=45
  swing2Motion=Locked
  twistLimitDegrees=45
  twistMotion=Locked
 }
 {bone1Name=dyn_Belt_01_front_r, bone2Name=Belt_front_r_strap, swing1LimitDegrees=5, swing2LimitDegrees=25, twistLimitDegrees=2}
 {
  bone1Name=Belt_front_r_strap
  bone2Name=pelvis
  swing1LimitDegrees=45
  swing1Motion=Locked
  swing2LimitDegrees=45
  swing2Motion=Locked
  twistLimitDegrees=45
  twistMotion=Locked
 }
 {bone1Name=dyn_Belt_02_front_r, bone2Name=dyn_Belt_01_front_r, swing1LimitDegrees=10, swing2LimitDegrees=35, twistLimitDegrees=2}
 {bone1Name=Base_Belt_01_backside_l, bone2Name=pelvis, swing1LimitDegrees=0, swing2LimitDegrees=0, twistLimitDegrees=0}
 {bone1Name=Base_Belt_01_backside_r, bone2Name=pelvis, swing1LimitDegrees=0, swing2LimitDegrees=0, twistLimitDegrees=0}
 {bone1Name=dyn_Belt_06_backside_l, bone2Name=dyn_Belt_05_backside_l, swing1LimitDegrees=35, swing2LimitDegrees=35, twistLimitDegrees=1}
 {bone1Name=dyn_Belt_07_backside_r, bone2Name=dyn_Belt_06_backside_r, swing1LimitDegrees=20, swing2LimitDegrees=20, twistLimitDegrees=1}
 {bone1Name=dyn_Belt_06_front_l, bone2Name=dyn_Belt_05_front_l, swing1LimitDegrees=35, swing2LimitDegrees=35, twistLimitDegrees=1}
 {bone1Name=dyn_Belt_07_front_r, bone2Name=dyn_Belt_06_front_r, swing1LimitDegrees=20, swing2LimitDegrees=20, twistLimitDegrees=1}
 {
  bone1Name=dyn_Belt_07_front_r
  bone2Name=dyn_Belt_07_backside_r
  swing1LimitDegrees=1
  swing1Motion=Locked
  swing2LimitDegrees=1
  swing2Motion=Locked
  twistLimitDegrees=1
  twistMotion=Locked
 }
 {bone1Name=dyn_Belt_06_front_l, bone2Name=dyn_Belt_06_backside_l, swing1LimitDegrees=1, swing2LimitDegrees=1, twistLimitDegrees=1}
]
