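# Jacket physics asset for the SmokeyAlias male medium player body, in a variant named for the DBNO (downed) state.
# Limited-angle chains swing the front and back jacket flaps from locked base bodies on spine_03, with tight constraints tying each front hem to its back side panel.
# Capsules on spine, pelvis and thighs form the body volume around the cloth.
# Front belt chains hang from the pelvis with all angular motion locked.
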PhysicsAsset /BRCosmetics/Characters/Player/Male/Medium/Bodies/M_MED_SmokeyAlias/Meshes/M_MED_SmokeyAlias_Physics_jaket_DBNO
bodies {
 Jacket_base_l [{shapeType=Capsule, length=0, radius=0, rotation=<rot -87.6388 0 0>}]
 Jacket_base_r [{shapeType=Capsule, length=0, radius=0}]
 Jacketback_base_backside [{shapeType=Capsule, length=0.505, radius=0.505, shapeName=Jacketback_base_backside_capsule, rotation=<rot 0 -90 0>}]
 Jacketback_l_base [{shapeType=Capsule, length=0, radius=0}]
 Jacketback_r_base [{shapeType=Capsule, length=0, radius=0}]
 dyn_Belt_01_front_l [{
   shapeType=Capsule
   length=3.6248
   radius=2.3616
   shapeName=dyn_Belt_01_front_l_capsule
   center=<1.6365 0.0908 -0.0643>
   rotation=<rot 0 178.289 -87.6878>
  }]
 dyn_Belt_01_front_r [{
   shapeType=Capsule
   length=3.7252
   radius=2.8517
   shapeName=dyn_Belt_01_front_r_capsule
   center=<1.0904 -0.1795 -0.0075>
   rotation=<rot 0 1.9464 -89.4074>
  }]
 dyn_Belt_02_front_l [{
   shapeType=Capsule
   length=6.7085
   radius=2.5328
   shapeName=dyn_Belt_02_front_l_capsule
   center=<3.4464 0.1219 0.6496>
   rotation=<rot 0 86.8295 -86.5933>
  }]
 dyn_Belt_02_front_r [{
   shapeType=Capsule
   length=4.9592
   radius=2.1819
   shapeName=dyn_Belt_02_front_r_capsule
   center=<3.6175 -0.4878 0.0946>
   rotation=<rot 0 83.6724 -87.9356>
  }]
 dyn_Belt_04_front_r [{
   shapeType=Capsule
   length=8.6574
   radius=2.6067
   shapeName=dyn_Belt_04_front_r_capsule
   center=<2.5789 0.0774 0.5818>
   rotation=<rot -0.8291 84.7378 -89.5214>
  }]
 dyn_Jacket_l_01 [{shapeType=Capsule, length=5.2644, radius=4.7851, center=<4 0 -4.2>, rotation=<rot 0 0 -90>}]
 dyn_Jacket_l_02 [
  {shapeType=Capsule, length=12.652, radius=4.6592, center=<3.76 0 -2.8469>, rotation=<rot -0.0001 1.2289 89.9999>}
  {
   shapeType=Capsule
   length=15.1494
   radius=4.4829
   shapeName=dyn_Jacket_l_02_capsule
   center=<0 -4.5128 -1.5162>
   rotation=<rot -85 0.0001 -0.0004>
  }
 ]
 dyn_Jacket_l_03 [{
   shapeType=Capsule
   length=15.4378
   radius=3.5953
   shapeName=dyn_Jacket_l_03_capsule
   center=<0.457 0.1206 0.0213>
   rotation=<rot -0.0002 -0.0004 79.9998>
  }]
 dyn_Jacket_l_05 [{shapeType=Capsule, length=0, radius=0, shapeName=dyn_Jacket_l_05_capsule, rotation=<rot 0 -90 0>}]
 dyn_Jacket_l_06 [
  {
   shapeType=Capsule
   length=4.1601
   radius=1.9915
   shapeName=dyn_Jacket_l_06_capsule
   center=<3 0 -1.4>
   rotation=<rot 3.0752 -92.1227 -91.8271>
  }
  {
   shapeType=Capsule
   length=9.614
   radius=4.3539
   shapeName=dyn_Jacket_l_06_capsule
   center=<-1.2386 -5.5866 -1.8395>
   rotation=<rot -2.1109 -24.9165 -89.5335>
  }
 ]
 dyn_Jacket_r_01 [{shapeType=Capsule, length=5.5533, radius=4.7851, center=<4 0 -4.2>, rotation=<rot 0 0 -90>}]
 dyn_Jacket_r_02 [{shapeType=Capsule, length=9.0248, radius=4.5926, center=<3.76 0 -3>, rotation=<rot 0 1.2292 80>}]
 dyn_Jacket_r_03 [{shapeType=Capsule, length=15.7971, radius=3.3499, shapeName=ja, center=<1.0874 0 -0.6474>, rotation=<rot -0.5123 -4.9734 84.1359>}]
 dyn_Jacket_r_05 [{shapeType=Capsule, length=0.5198, radius=0.5, rotation=<rot 0 -90 0>}]
 dyn_Jacket_r_06 [{shapeType=Capsule, length=6.0953, radius=1.8061, center=<2.146 -1.456 -0.9217>, rotation=<rot 3.0752 -92.1227 -91.8271>}]
 dyn_Jacketback_backside_01 [{
   shapeType=Capsule
   length=4.8752
   radius=3.6894
   shapeName=dyn_Jacketback_backside_01_capsule
   center=<3.33 0 -2.5>
   rotation=<rot 0.0001 0 90>
  }]
 dyn_Jacketback_backside_02 [{
   shapeType=Capsule
   length=13.4702
   radius=3.5377
   shapeName=dyn_Jacketback_backside_02_capsule
   center=<5.5 0 -3>
   rotation=<rot 0 -90 94.9996>
  }]
 dyn_Jacketback_l_01 [{shapeType=Capsule, length=3.2526, radius=5.0474, center=<4.6 0 -3>, rotation=<rot 0.0001 -0.0002 -73.2154>}]
 dyn_Jacketback_l_02 [{shapeType=Capsule, length=6.1593, radius=5.6619, center=<6.8405 0.4494 -5.2703>, rotation=<rot 0.0002 -0.0004 -83.3745>}]
 dyn_Jacketback_r_01 [{shapeType=Capsule, length=3.2727, radius=5.0273, center=<4.6 0 -3>, rotation=<rot 0.0001 -0.0002 -115.1285>}]
 dyn_Jacketback_r_02 [{shapeType=Capsule, length=5.7197, radius=6.1015, center=<6.7 0 -6>, rotation=<rot 0.0001 0.0002 88.3745>}]
 pelvis [
  {shapeType=Capsule, length=0.6904, radius=11.5295, center=<0 -0.9558 0>, rotation=<rot 0 -0.0003 0>}
  {shapeType=Capsule, length=11.3214, radius=6.8176, shapeName=pelvis_capsule, center=<0.4598 6.3326 0>}
 ]
 spine_03 [{shapeType=Capsule, length=18.5478, radius=11.9148, center=<-13.5302 0 0>, rotation=<rot 90 -179.9999 -179.9995>}]
 thigh_l [{
   shapeType=Capsule
   length=11.7384
   radius=11.0383
   shapeName=thigh_l_capsule
   center=<-9.7677 0.6827 0.7691>
   rotation=<rot 87.5546 -142.2235 -142.1979>
  }]
 thigh_r [{
   shapeType=Capsule
   length=11.7384
   radius=11.0383
   shapeName=thigh_l_capsule
   center=<9.8222 0.7077 -0.19>
   rotation=<rot 87.5546 -142.2238 -142.1981>
  }]
 thigh_twist_01_l [{shapeType=Capsule, length=15.0979, radius=7.5015, center=<-1.4228 0 0>, rotation=<rot -90 -179.9852 179.985>}]
 thigh_twist_01_r [{shapeType=Capsule, length=13.8336, radius=8.7658, center=<1.4228 0 0>, rotation=<rot -90 -179.9701 -0.0304>}]
}
constraints [
 {
  bone1Name=spine_03
  bone2Name=pelvis
  swing1LimitDegrees=45
  swing1Motion=Locked
  swing2LimitDegrees=45
  swing2Motion=Locked
  twistLimitDegrees=45
  twistMotion=Locked
 }
 {
  bone1Name=thigh_l
  bone2Name=pelvis
  swing1LimitDegrees=45
  swing1Motion=Locked
  swing2LimitDegrees=45
  swing2Motion=Locked
  twistLimitDegrees=45
  twistMotion=Locked
 }
 {
  bone1Name=thigh_r
  bone2Name=pelvis
  swing1LimitDegrees=45
  swing1Motion=Locked
  swing2LimitDegrees=45
  swing2Motion=Locked
  twistLimitDegrees=45
  twistMotion=Locked
 }
 {
  bone1Name=Jacket_base_l
  bone2Name=spine_03
  swing1LimitDegrees=45
  swing1Motion=Locked
  swing2LimitDegrees=45
  swing2Motion=Locked
  twistLimitDegrees=45
  twistMotion=Locked
 }
 {
  bone1Name=Jacket_base_r
  bone2Name=spine_03
  swing1LimitDegrees=45
  swing1Motion=Locked
  swing2LimitDegrees=45
  swing2Motion=Locked
  twistLimitDegrees=45
  twistMotion=Locked
 }
 {bone1Name=dyn_Jacket_l_01, bone2Name=Jacket_base_l, swing1LimitDegrees=45, swing2LimitDegrees=45, twistLimitDegrees=15}
 {bone1Name=dyn_Jacket_l_02, bone2Name=dyn_Jacket_l_01, swing1LimitDegrees=45, swing2LimitDegrees=45, twistLimitDegrees=15}
 {bone1Name=dyn_Jacket_r_01, bone2Name=Jacket_base_r, swing1LimitDegrees=45, swing2LimitDegrees=45, twistLimitDegrees=10}
 {bone1Name=dyn_Jacket_r_02, bone2Name=dyn_Jacket_r_01, swing1LimitDegrees=30, swing2LimitDegrees=30, twistLimitDegrees=10}
 {bone1Name=dyn_Jacket_r_03, bone2Name=dyn_Jacket_r_02, swing1LimitDegrees=45, swing2LimitDegrees=45, twistLimitDegrees=10}
 {bone1Name=dyn_Jacket_r_06, bone2Name=dyn_Jacket_r_05, swing1LimitDegrees=10, swing2LimitDegrees=10, twistLimitDegrees=5}
 {bone1Name=dyn_Jacketback_r_02, bone2Name=dyn_Jacketback_r_01, swing1LimitDegrees=10, swing2LimitDegrees=20, twistLimitDegrees=5}
 {
  bone1Name=Jacketback_l_base
  bone2Name=spine_03
  swing1LimitDegrees=45
  swing1Motion=Locked
  swing2LimitDegrees=45
  swing2Motion=Locked
  twistLimitDegrees=45
  twistMotion=Locked
 }
 {bone1Name=dyn_Jacketback_l_01, bone2Name=Jacketback_l_base, swing1LimitDegrees=10, swing2LimitDegrees=10, twistLimitDegrees=2}
 {bone1Name=dyn_Jacketback_l_02, bone2Name=dyn_Jacketback_l_01, swing1LimitDegrees=10, swing2LimitDegrees=20, twistLimitDegrees=5}
 {
  bone1Name=Jacketback_r_base
  bone2Name=spine_03
  swing1LimitDegrees=45
  swing1Motion=Locked
  swing2LimitDegrees=45
  swing2Motion=Locked
  twistLimitDegrees=45
  twistMotion=Locked
 }
 {bone1Name=dyn_Jacketback_r_01, bone2Name=Jacketback_r_base, swing1LimitDegrees=10, swing2LimitDegrees=10, twistLimitDegrees=2}
 {
  bone1Name=dyn_Jacket_r_05
  bone2Name=dyn_Jacket_r_03
  swing1LimitDegrees=45
  swing1Motion=Locked
  swing2LimitDegrees=45
  swing2Motion=Locked
  twistLimitDegrees=45
  twistMotion=Locked
 }
 {bone1Name=dyn_Jacket_l_03, bone2Name=dyn_Jacket_l_02, swing1LimitDegrees=45, swing2LimitDegrees=45, twistLimitDegrees=10}
 {
  bone1Name=dyn_Jacket_l_05
  bone2Name=dyn_Jacket_l_03
  swing1LimitDegrees=45
  swing1Motion=Locked
  swing2LimitDegrees=45
  swing2Motion=Locked
  twistLimitDegrees=45
  twistMotion=Locked
 }
 {bone1Name=dyn_Jacket_l_06, bone2Name=dyn_Jacket_l_05, swing1LimitDegrees=10, swing2LimitDegrees=10, twistLimitDegrees=5}
 {
  bone1Name=thigh_twist_01_l
  bone2Name=thigh_l
  swing1LimitDegrees=45
  swing1Motion=Locked
  swing2LimitDegrees=45
  swing2Motion=Locked
  twistLimitDegrees=45
  twistMotion=Locked
 }
 {
  bone1Name=thigh_twist_01_r
  bone2Name=thigh_r
  swing1LimitDegrees=45
  swing1Motion=Locked
  swing2LimitDegrees=45
  swing2Motion=Locked
  twistLimitDegrees=45
  twistMotion=Locked
 }
 {
  bone1Name=Jacketback_base_backside
  bone2Name=spine_03
  swing1LimitDegrees=45
  swing1Motion=Locked
  swing2LimitDegrees=45
  swing2Motion=Locked
  twistLimitDegrees=45
  twistMotion=Locked
 }
 {bone1Name=dyn_Jacketback_backside_01, bone2Name=Jacketback_base_backside, swing1LimitDegrees=5, swing2LimitDegrees=15, twistLimitDegrees=1}
 {bone1Name=dyn_Jacketback_backside_02, bone2Name=dyn_Jacketback_backside_01, swing1LimitDegrees=5, swing2LimitDegrees=30, twistLimitDegrees=1}
 {bone1Name=dyn_Jacket_l_03, bone2Name=dyn_Jacketback_l_02, swing1LimitDegrees=1, swing2LimitDegrees=1, twistLimitDegrees=1}
 {bone1Name=dyn_Jacket_r_03, bone2Name=dyn_Jacketback_r_02, swing1LimitDegrees=1, swing2LimitDegrees=1, twistLimitDegrees=1}
 {
  bone1Name=dyn_Belt_01_front_l
  bone2Name=pelvis
  swing1LimitDegrees=45
  swing1Motion=Locked
  swing2LimitDegrees=45
  swing2Motion=Locked
  twistLimitDegrees=45
  twistMotion=Locked
 }
 {
  bone1Name=dyn_Belt_02_front_l
  bone2Name=dyn_Belt_01_front_l
  swing1LimitDegrees=45
  swing1Motion=Locked
  swing2LimitDegrees=45
  swing2Motion=Locked
  twistLimitDegrees=45
  twistMotion=Locked
 }
 {
  bone1Name=dyn_Belt_01_front_r
  bone2Name=pelvis
  swing1LimitDegrees=45
  swing1Motion=Locked
  swing2LimitDegrees=45
  swing2Motion=Locked
  twistLimitDegrees=45
  twistMotion=Locked
 }
 {
  bone1Name=dyn_Belt_02_front_r
  bone2Name=dyn_Belt_01_front_r
  swing1LimitDegrees=45
  swing1Motion=Locked
  swing2LimitDegrees=45
  swing2Motion=Locked
  twistLimitDegrees=45
  twistMotion=Locked
 }
 {
  bone1Name=dyn_Belt_04_front_r
  bone2Name=dyn_Belt_02_front_r
  swing1LimitDegrees=45
  swing1Motion=Locked
  swing2LimitDegrees=45
  swing2Motion=Locked
  twistLimitDegrees=45
  twistMotion=Locked
 }
]
