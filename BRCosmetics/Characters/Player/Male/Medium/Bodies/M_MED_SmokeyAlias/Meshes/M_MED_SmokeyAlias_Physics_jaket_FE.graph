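# Physics asset for the jacket of the SmokeyAlias medium male body cosmetic, with capsule bodies for secondary jacket motion.
# Capsules on the pelvis, spine and thighs form the body volume, with jacket base bones locked to spine_03.
# Front and back jacket chains hang from those bases with tight angular limits, though the right front tip is fully locked, unlike the left.
# Front belt chains and the bow are locked back to the pelvis.
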PhysicsAsset /BRCosmetics/Characters/Player/Male/Medium/Bodies/M_MED_SmokeyAlias/Meshes/M_MED_SmokeyAlias_Physics_jaket_FE
bodies {
 Jacket_base_l [{shapeType=Capsule, length=0, radius=0, rotation=<rot -87.6388 0 0>}]
 Jacket_base_r [{shapeType=Capsule, length=0, radius=0}]
 Jacketback_base_backside [{shapeType=Capsule, length=0.505, radius=0.505, shapeName=Jacketback_base_backside_capsule, rotation=<rot 0 -90 0>}]
 Jacketback_l_base [{shapeType=Capsule, length=0, radius=0}]
 Jacketback_r_base [{shapeType=Capsule, length=0, radius=0}]
 bow [{shapeType=Capsule, length=0, radius=0.5, shapeName=bow_capsule, center=<0.6546 0.2698 0.0285>, rotation=<rot 0 88.8257 -89.4259>}]
 dyn_Belt_01_front_l [{
   shapeType=Capsule
   length=4.3
   radius=1
   shapeName=dyn_Belt_01_front_l_capsule
   center=<1.6358 0.0671 0.5234>
   rotation=<rot 0 178.289 -87.6878>
  }]
 dyn_Belt_01_front_r [{
   shapeType=Capsule
   length=3.9195
   radius=2.8591
   shapeName=dyn_Belt_01_front_r_capsule
   center=<1.392 -0.1485 1.5579>
   rotation=<rot 0 -0.2681 -89.1154>
  }]
 dyn_Belt_02_front_l [{
   shapeType=Capsule
   length=3.697
   radius=3.9465
   shapeName=dyn_Belt_02_front_l_capsule
   center=<3.2616 0.1171 2.8375>
   rotation=<rot 0 86.7004 -86.2624>
  }]
 dyn_Belt_02_front_r [{
   shapeType=Capsule
   length=3.6382
   radius=2.4342
   shapeName=dyn_Belt_02_front_r_capsule
   center=<2.8915 -0.4342 1.3034>
   rotation=<rot 0 88.9387 -82.8563>
  }]
 dyn_Belt_03_front_l [{
   shapeType=Capsule
   length=7.1715
   radius=2.3767
   shapeName=dyn_Belt_03_front_l_capsule
   center=<1.7739 -0.0495 1.307>
   rotation=<rot -0.6326 89.6779 -89.6886>
  }]
 dyn_Belt_04_front_r [{
   shapeType=Capsule
   length=5.6725
   radius=2.0641
   shapeName=dyn_Belt_04_front_r_capsule
   center=<1.8546 -0.2866 1.4589>
   rotation=<rot 0.9909 -92.3605 -89.8798>
  }]
 dyn_Jacket_l_01 [
  {shapeType=Capsule, length=7.918, radius=4.0724, center=<4 0 -3.4>, rotation=<rot 0 4.107 -90>}
  {
   shapeType=Capsule
   length=9.6104
   radius=3.5303
   shapeName=dyn_Jacket_l_01_capsule
   center=<4.4516 -6.4 -2.1709>
   rotation=<rot -89.878 179.9903 -174.9908>
  }
 ]
 dyn_Jacket_l_02 [
  {shapeType=Capsule, length=8.9327, radius=3.8696, center=<3.76 0 -3>, rotation=<rot -0.0001 1.2289 89.9999>}
  {shapeType=Capsule, length=9.0113, radius=3.4887, shapeName=dyn_Jacket_l_02_capsule, center=<3 -5.9 -2>, rotation=<rot -90 0 -0.0001>}
 ]
 dyn_Jacket_l_03 [{shapeType=Capsule, length=10.0744, radius=2.4, shapeName=dyn_Jacket_l_03_capsule, center=<3 0 0.6536>, rotation=<rot 0 -0.0001 80>}]
 dyn_Jacket_l_05 [{shapeType=Capsule, length=0, radius=0.5, shapeName=dyn_Jacket_l_05_capsule, rotation=<rot 0 -90 0>}]
 dyn_Jacket_l_06 [{
   shapeType=Capsule
   length=3.176
   radius=1.9755
   shapeName=dyn_Jacket_l_06_capsule
   center=<3 0 -1.4>
   rotation=<rot 3.0752 -92.1227 -91.8271>
  }]
 dyn_Jacket_r_01 [
  {shapeType=Capsule, length=5.8381, radius=3.9798, center=<4 0 -4.2>, rotation=<rot 0 -4.9999 -90>}
  {
   shapeType=Capsule
   length=6.9857
   radius=3.1595
   shapeName=dyn_Jacket_r_01_capsule
   center=<4.3 6.7916 -1.62>
   rotation=<rot -90 -179.9999 -179.9999>
  }
 ]
 dyn_Jacket_r_02 [
  {shapeType=Capsule, length=6.857, radius=4.4, center=<3.76 0 -3>, rotation=<rot 0 1.2292 80>}
  {
   shapeType=Capsule
   length=7.181
   radius=3.4
   shapeName=dyn_Jacket_r_02_capsule
   center=<2.1417 7.125 -0.822>
   rotation=<rot -90 -179.9901 179.9902>
  }
 ]
 dyn_Jacket_r_03 [{shapeType=Capsule, length=11.9327, radius=3.4181, shapeName=ja, center=<3 0 -0.7>, rotation=<rot -0.0001 0.0002 84.1136>}]
 dyn_Jacket_r_05 [{shapeType=Capsule, length=0.5198, radius=0.5, rotation=<rot 0 -90 0>}]
 dyn_Jacket_r_06 [{shapeType=Capsule, length=6.3453, radius=1.5561, center=<2.2 0 -1>, rotation=<rot 3.0752 -92.1227 -91.8271>}]
 dyn_Jacketback_backside_01 [{
   shapeType=Capsule
   length=4.8752
   radius=3.6894
   shapeName=dyn_Jacketback_backside_01_capsule
   center=<3.33 0 -2.5>
   rotation=<rot 0.0001 0 90>
  }]
 dyn_Jacketback_backside_02 [{
   shapeType=Capsule
   length=13.4702
   radius=3.5377
   shapeName=dyn_Jacketback_backside_02_capsule
   center=<5.5 0 -3>
   rotation=<rot 0 -90 94.9996>
  }]
 dyn_Jacketback_l_01 [{shapeType=Capsule, length=5.3694, radius=2.9306, center=<4.6 0 -4>, rotation=<rot 0.0001 -0.0002 -73.2154>}]
 dyn_Jacketback_l_02 [{shapeType=Capsule, length=9.8298, radius=5.1702, center=<6.8405 0 -5.55>, rotation=<rot 0.0001 -0.0002 -68.3745>}]
 dyn_Jacketback_r_01 [{shapeType=Capsule, length=5.3289, radius=2.9711, center=<4.6 0 -3>, rotation=<rot 0.0001 -0.0002 -115.1285>}]
 dyn_Jacketback_r_02 [{shapeType=Capsule, length=9.7359, radius=5.7962, center=<6.7 0 -6.24>, rotation=<rot 0.0001 0.0001 83.3745>}]
 pelvis [
  {shapeType=Capsule, length=6.0918, radius=11.5295, center=<0 -0.9558 0>, rotation=<rot 0 -0.0003 0>}
  {shapeType=Capsule, length=9.3496, radius=6.8176, shapeName=pelvis_capsule, center=<0.4598 6.3326 0>}
  {shapeType=Capsule, length=12.9128, radius=7.4093, shapeName=pelvis_capsule, center=<2.322 0 0.9833>}
 ]
 spine_01 [{shapeType=Capsule, length=6.5568, radius=12.2261, shapeName=spine_01_capsule, rotation=<rot 90 0 0>}]
 spine_02 [{shapeType=Capsule, length=8.7838, radius=12.0119, shapeName=spine_02_capsule, rotation=<rot -90 179.9992 -179.9993>}]
 spine_03 [{shapeType=Capsule, length=4.892, radius=12.1107, rotation=<rot 90 -179.9999 -179.9995>}]
 thigh_l [{
   shapeType=Capsule
   length=13.0375
   radius=9
   shapeName=thigh_l_capsule
   center=<-9.7677 0.6827 0.7691>
   rotation=<rot 87.5546 -142.2235 -142.1979>
  }]
 thigh_r [{
   shapeType=Capsule
   length=18.3422
   radius=8.4689
   shapeName=thigh_l_capsule
   center=<9.8222 0.7077 -0.19>
   rotation=<rot 87.5546 -142.2238 -142.1981>
  }]
 thigh_twist_01_l [{shapeType=Capsule, length=14.3606, radius=8.2388, center=<-1.4228 0 0>, rotation=<rot -90 -179.9852 179.985>}]
 thigh_twist_01_r [{shapeType=Capsule, length=14.6879, radius=7.9115, center=<1.4228 0 0>, rotation=<rot -90 -179.9701 -0.0304>}]
}
constraints [
 {
  bone1Name=spine_03
  bone2Name=pelvis
  swing1LimitDegrees=45
  swing1Motion=Locked
  swing2LimitDegrees=45
  swing2Motion=Locked
  twistLimitDegrees=45
  twistMotion=Locked
 }
 {
  bone1Name=thigh_l
  bone2Name=pelvis
  swing1LimitDegrees=45
  swing1Motion=Locked
  swing2LimitDegrees=45
  swing2Motion=Locked
  twistLimitDegrees=45
  twistMotion=Locked
 }
 {
  bone1Name=thigh_r
  bone2Name=pelvis
  swing1LimitDegrees=45
  swing1Motion=Locked
  swing2LimitDegrees=45
  swing2Motion=Locked
  twistLimitDegrees=45
  twistMotion=Locked
 }
 {
  bone1Name=Jacket_base_l
  bone2Name=spine_03
  swing1LimitDegrees=45
  swing1Motion=Locked
  swing2LimitDegrees=45
  swing2Motion=Locked
  twistLimitDegrees=45
  twistMotion=Locked
 }
 {
  bone1Name=Jacket_base_r
  bone2Name=spine_03
  swing1LimitDegrees=45
  swing1Motion=Locked
  swing2LimitDegrees=45
  swing2Motion=Locked
  twistLimitDegrees=45
  twistMotion=Locked
 }
 {bone1Name=dyn_Jacket_l_01, bone2Name=Jacket_base_l, swing1LimitDegrees=2, swing2LimitDegrees=20, twistLimitDegrees=30}
 {bone1Name=dyn_Jacket_l_02, bone2Name=dyn_Jacket_l_01, swing1LimitDegrees=10, swing2LimitDegrees=25, twistLimitDegrees=10}
 {bone1Name=dyn_Jacket_r_01, bone2Name=Jacket_base_r, swing1LimitDegrees=10, swing2LimitDegrees=35, twistLimitDegrees=20}
 {bone1Name=dyn_Jacket_r_02, bone2Name=dyn_Jacket_r_01, swing1LimitDegrees=15, swing2LimitDegrees=26, twistLimitDegrees=10}
 {bone1Name=dyn_Jacket_r_03, bone2Name=dyn_Jacket_r_02, swing1LimitDegrees=5, swing2LimitDegrees=10, twistLimitDegrees=5}
 {
  bone1Name=dyn_Jacket_r_06
  bone2Name=dyn_Jacket_r_05
  swing1LimitDegrees=15
  swing1Motion=Locked
  swing2LimitDegrees=10
  swing2Motion=Locked
  twistLimitDegrees=3
  twistMotion=Locked
 }
 {bone1Name=dyn_Jacketback_r_02, bone2Name=dyn_Jacketback_r_01, swing1LimitDegrees=5, swing2LimitDegrees=10, twistLimitDegrees=5}
 {
  bone1Name=Jacketback_l_base
  bone2Name=spine_03
  swing1LimitDegrees=45
  swing1Motion=Locked
  swing2LimitDegrees=45
  swing2Motion=Locked
  twistLimitDegrees=45
  twistMotion=Locked
 }
 {bone1Name=dyn_Jacketback_l_01, bone2Name=Jacketback_l_base, swing1LimitDegrees=10, swing2LimitDegrees=20, twistLimitDegrees=2}
 {bone1Name=dyn_Jacketback_l_02, bone2Name=dyn_Jacketback_l_01, swing1LimitDegrees=5, swing2LimitDegrees=10, twistLimitDegrees=5}
 {
  bone1Name=Jacketback_r_base
  bone2Name=spine_03
  swing1LimitDegrees=45
  swing1Motion=Locked
  swing2LimitDegrees=45
  swing2Motion=Locked
  twistLimitDegrees=45
  twistMotion=Locked
 }
 {bone1Name=dyn_Jacketback_r_01, bone2Name=Jacketback_r_base, swing1LimitDegrees=10, swing2LimitDegrees=20, twistLimitDegrees=10}
 {
  bone1Name=dyn_Jacket_r_05
  bone2Name=dyn_Jacket_r_03
  swing1LimitDegrees=45
  swing1Motion=Locked
  swing2LimitDegrees=45
  swing2Motion=Locked
  twistLimitDegrees=45
  twistMotion=Locked
 }
 {bone1Name=dyn_Jacket_l_03, bone2Name=dyn_Jacket_l_02, swing1LimitDegrees=5, swing2LimitDegrees=14, twistLimitDegrees=5}
 {
  bone1Name=dyn_Jacket_l_05
  bone2Name=dyn_Jacket_l_03
  swing1LimitDegrees=45
  swing1Motion=Locked
  swing2LimitDegrees=45
  swing2Motion=Locked
  twistLimitDegrees=45
  twistMotion=Locked
 }
 {bone1Name=dyn_Jacket_l_06, bone2Name=dyn_Jacket_l_05, swing1LimitDegrees=15, swing2LimitDegrees=25, twistLimitDegrees=0}
 {
  bone1Name=thigh_twist_01_l
  bone2Name=thigh_l
  swing1LimitDegrees=45
  swing1Motion=Locked
  swing2LimitDegrees=45
  swing2Motion=Locked
  twistLimitDegrees=45
  twistMotion=Locked
 }
 {
  bone1Name=thigh_twist_01_r
  bone2Name=thigh_r
  swing1LimitDegrees=45
  swing1Motion=Locked
  swing2LimitDegrees=45
  swing2Motion=Locked
  twistLimitDegrees=45
  twistMotion=Locked
 }
 {
  bone1Name=dyn_Belt_01_front_l
  bone2Name=pelvis
  swing1LimitDegrees=45
  swing1Motion=Locked
  swing2LimitDegrees=45
  swing2Motion=Locked
  twistLimitDegrees=45
  twistMotion=Locked
 }
 {
  bone1Name=dyn_Belt_02_front_l
  bone2Name=dyn_Belt_01_front_l
  swing1LimitDegrees=45
  swing1Motion=Locked
  swing2LimitDegrees=45
  swing2Motion=Locked
  twistLimitDegrees=45
  twistMotion=Locked
 }
 {
  bone1Name=dyn_Belt_03_front_l
  bone2Name=dyn_Belt_02_front_l
  swing1LimitDegrees=45
  swing1Motion=Locked
  swing2LimitDegrees=45
  swing2Motion=Locked
  twistLimitDegrees=45
  twistMotion=Locked
 }
 {
  bone1Name=dyn_Belt_01_front_r
  bone2Name=pelvis
  swing1LimitDegrees=45
  swing1Motion=Locked
  swing2LimitDegrees=45
  swing2Motion=Locked
  twistLimitDegrees=45
  twistMotion=Locked
 }
 {
  bone1Name=dyn_Belt_02_front_r
  bone2Name=dyn_Belt_01_front_r
  swing1LimitDegrees=45
  swing1Motion=Locked
  swing2LimitDegrees=45
  swing2Motion=Locked
  twistLimitDegrees=45
  twistMotion=Locked
 }
 {
  bone1Name=dyn_Belt_04_front_r
  bone2Name=dyn_Belt_02_front_r
  swing1LimitDegrees=45
  swing1Motion=Locked
  swing2LimitDegrees=45
  swing2Motion=Locked
  twistLimitDegrees=45
  twistMotion=Locked
 }
 {
  bone1Name=Jacketback_base_backside
  bone2Name=spine_03
  swing1LimitDegrees=45
  swing1Motion=Locked
  swing2LimitDegrees=45
  swing2Motion=Locked
  twistLimitDegrees=45
  twistMotion=Locked
 }
 {bone1Name=dyn_Jacketback_backside_01, bone2Name=Jacketback_base_backside, swing1LimitDegrees=5, swing2LimitDegrees=15, twistLimitDegrees=1}
 {bone1Name=dyn_Jacketback_backside_02, bone2Name=dyn_Jacketback_backside_01, swing1LimitDegrees=5, swing2LimitDegrees=30, twistLimitDegrees=1}
 {
  bone1Name=bow
  bone2Name=pelvis
  swing1LimitDegrees=45
  swing1Motion=Locked
  swing2LimitDegrees=45
  swing2Motion=Locked
  twistLimitDegrees=45
  twistMotion=Locked
 }
]
