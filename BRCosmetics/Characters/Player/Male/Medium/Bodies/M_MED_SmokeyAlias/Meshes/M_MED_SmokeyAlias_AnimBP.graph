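# Character-part AnimBP for the male SmokeyAlias body, adding secondary motion and corrective rigging to a pose copied from the parent mesh.
# AnimDynamics moves collar and shoe bones, then a state-driven Control Rig and bone constraints produce a cached pre-simulation pose.
# Chained RigidBody nodes simulate belts and jacket, switching physics assets for front-end, DBNO or normal play.
# The EventGraph updates gravity override, crouch, movement, LOD and non-emote front-end values.
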AnimBlueprint /BRCosmetics/Characters/Player/Male/Medium/Bodies/M_MED_SmokeyAlias/Meshes/M_MED_SmokeyAlias_AnimBP
target_skeleton=&/M_MED_SmokeyAlias_Skeleton
graphs [
 AnimGraph {nodes [
   Blend Poses by bool(BlendPose_0=Component To Local(), bActiveValue=Get bIsCrouching()) {blendType=Linear}
   n1 = Control Rig(Source=Copy Pose From Mesh() {bUseAttachedParent=true}) {
    alphaBoolBlend {blendOption=Linear}
    controlRigAssetReference {blueprintRigClass=&/M_MED_SmokeyAlias_CtrlRig.M_MED_SmokeyAlias_CtrlRig_C}
   }
   Local To Component(LocalPose=@n1)
   n2 = Get Gravity Override()
   n3 = RigidBody!(ComponentPose=Local To Component(LocalPose=Use cached pose 'presim'()), Alpha, OverrideWorldGravity:Vector=@n2, ExternalForce) -> (Pose) {
    bClampLinearTranslationLimitToRefPose=true
    bOverrideWorldGravity=true
    overridePhysicsAsset=&/M_MED_SmokeyAlias_Physics_belts
    simulationSpace=BaseBoneSpace
    alphaBoolBlend {blendOption=Linear}
    baseBoneRef {boneName=pelvis}
    overrideWorldGravity {z=-980}
    simSpaceSettings {
     maxAngularAcceleration=500
     maxAngularVelocity=3
     maxLinearAcceleration=450
     maxLinearVelocity=250
     velocityScaleZ=0
     worldAlpha=1
     externalLinearDragV=<0.1 0.1>
    }
   }
   n4 = RigidBody(ComponentPose=@n3, OverrideWorldGravity=@n2) {
    bClampLinearTranslationLimitToRefPose=true
    bOverrideWorldGravity=true
    overridePhysicsAsset=&/M_MED_SmokeyAlias_Physics_jaket_DBNO
    simulationSpace=BaseBoneSpace
    alphaBoolBlend {blendOption=Linear}
    baseBoneRef {boneName=pelvis}
    overrideWorldGravity {z=-980}
    simSpaceSettings {
     maxAngularAcceleration=500
     maxAngularVelocity=3
     maxLinearAcceleration=450
     maxLinearVelocity=250
     velocityScaleZ=0
     worldAlpha=1
     externalLinearDragV=<0.1 0.1>
    }
   }
   n5 = Get Gravity Override()
   n6 = RigidBody(ComponentPose=Local To Component(LocalPose=Use cached pose 'presim'()), OverrideWorldGravity=@n5) {
    bClampLinearTranslationLimitToRefPose=true
    bOverrideWorldGravity=true
    overridePhysicsAsset=&M_MED_SmokeyAlias_Physics_belts
    simulationSpace=BaseBoneSpace
    alphaBoolBlend {blendOption=Linear}
    baseBoneRef {boneName=pelvis}
    overrideWorldGravity {z=-980}
    simSpaceSettings {
     maxAngularAcceleration=500
     maxAngularVelocity=3
     maxLinearAcceleration=450
     maxLinearVelocity=250
     velocityScaleZ=0
     worldAlpha=1
     externalLinearDragV=<0.1 0.1>
    }
   }
   n7 = RigidBody(ComponentPose=@n6, OverrideWorldGravity=@n5) {
    bClampLinearTranslationLimitToRefPose=true
    bOverrideWorldGravity=true
    overridePhysicsAsset=&/M_MED_SmokeyAlias_Physics_jaket
    simulationSpace=BaseBoneSpace
    alphaBoolBlend {blendOption=Linear}
    baseBoneRef {boneName=spine_03}
    overrideWorldGravity {z=-980}
    simSpaceSettings {
     maxAngularAcceleration=500
     maxAngularVelocity=3
     maxLinearAcceleration=450
     maxLinearVelocity=250
     velocityScaleZ=0
     worldAlpha=1
     externalLinearDragV=<0.1 0.1>
    }
   }
   n8 = Get Gravity Override()
   n9 = RigidBody(ComponentPose=Local To Component(LocalPose=Use cached pose 'presim'()), OverrideWorldGravity=@n8) {
    bClampLinearTranslationLimitToRefPose=true
    bOverrideWorldGravity=true
    overridePhysicsAsset=&M_MED_SmokeyAlias_Physics_belts
    simulationSpace=BaseBoneSpace
    alphaBoolBlend {blendOption=Linear}
    baseBoneRef {boneName=pelvis}
    overrideWorldGravity {z=-980}
    simSpaceSettings {
     maxAngularAcceleration=500
     maxAngularVelocity=3
     maxLinearAcceleration=450
     maxLinearVelocity=250
     velocityScaleZ=0
     worldAlpha=1
     externalLinearDragV=<0.1 0.1>
    }
   }
   n10 = RigidBody(ComponentPose=@n9, OverrideWorldGravity=@n8) {
    bClampLinearTranslationLimitToRefPose=true
    bOverrideWorldGravity=true
    overridePhysicsAsset=&/M_MED_SmokeyAlias_Physics_jaket_FE
    simulationSpace=BaseBoneSpace
    alphaBoolBlend {blendOption=Linear}
    baseBoneRef {boneName=spine_03}
    overrideWorldGravity {z=-980}
    simSpaceSettings {
     maxAngularAcceleration=500
     maxAngularVelocity=3
     maxLinearAcceleration=450
     maxLinearVelocity=250
     velocityScaleZ=0
     worldAlpha=1
     externalLinearDragV=<0.1 0.1>
    }
   }
   n11 = Blend Poses by bool(
    BlendPose_0=Component To Local(ComponentPose=@n10)
    BlendPose_1=Blend Poses by bool(BlendPose_0=Component To Local(ComponentPose=@n4), BlendPose_1=Component To Local(ComponentPose=@n7), bActiveValue=Get bIsDBNO()) {blendType=Linear}
    bActiveValue=Get is my front end() -> (is my front end:bool)
   ) {blendType=Linear}
   Output Pose(Result=Component To Local(ComponentPose=Local To Component(LocalPose=@n11))) {layerGroup=None}
   n12 = Anim Dynamics - Bone: dyn_collar_l(ComponentPose=Local To Component(LocalPose=Copy Pose From Mesh() {bUseAttachedParent=true}), Alpha=1, ExternalForce="0,0,0") {
    angularDampingOverride=1
    bAngularSpring=true
    bOverrideAngularDamping=true
    gravityScale=0.1
    linearSpringConstant=10
    simulationSpace=World
    alphaBoolBlend {blendOption=Linear}
    boundBone {boneName=dyn_collar_l}
    componentAppliedLinearAccClamp=<10000 10000 10000>
    componentLinearAccScale=<1 1 1>
    componentLinearVelScale=<1 1 1>
    physicsBodyDefinitions [{
      boundBone {boneName=dyn_collar_l}
      boxExtents=<10 8 8>
      constraintSetup {angularLimitsMax=<0 20 4>, angularLimitsMin=<0 -3 -2>, angularTarget=<0 0 0>}
      localJointOffset=<2 0 0>
     }]
   }
   n13 = Anim Dynamics - Bone: dyn_collar_r(ComponentPose=@n12, Alpha=1, ExternalForce="0,0,0") {
    angularDampingOverride=1
    angularSpringConstant=500
    bAngularSpring=true
    bOverrideAngularDamping=true
    gravityScale=0.1
    simulationSpace=World
    alphaBoolBlend {blendOption=Linear}
    boundBone {boneName=dyn_collar_r}
    componentAppliedLinearAccClamp=<10000 10000 10000>
    componentLinearAccScale=<1 1 1>
    componentLinearVelScale=<1 1 1>
    physicsBodyDefinitions [{
      boundBone {boneName=dyn_collar_r}
      boxExtents=<8 9 12>
      constraintSetup {angularLimitsMax=<0 20 2>, angularLimitsMin=<0 -3 -2>, angularTarget=<0 0 0>}
      localJointOffset=<2 0 0>
     }]
   }
   n14 = Anim Dynamics - Bone: dyn_shoes_r(ComponentPose=@n13, Alpha=1, ExternalForce="0,0,0") {
    angularDampingOverride=1
    angularSpringConstant=497
    bAngularSpring=true
    bOverrideAngularDamping=true
    bUseGravityOverride=true
    gravityScale=0.1
    simulationSpace=World
    alphaBoolBlend {blendOption=Linear}
    boundBone {boneName=dyn_shoes_r}
    componentAppliedLinearAccClamp=<10000 10000 10000>
    componentLinearAccScale=<1 1 1>
    componentLinearVelScale=<1 1 1>
    gravityOverride {z=-980}
    physicsBodyDefinitions [{
      boundBone {boneName=dyn_shoes_r}
      boxExtents=<1 1 1>
      constraintSetup {angularLimitsMax=<0 20 0>, angularLimitsMin=<0 -10 0>, angularTarget=<0 0 0>}
      localJointOffset=<0 0 0>
     }]
   }
   n15 = Anim Dynamics - Bone: dyn_shoes_l(ComponentPose=@n14, Alpha=1, ExternalForce="0,0,0") {
    angularDampingOverride=1
    angularSpringConstant=500
    bAngularSpring=true
    bOverrideAngularDamping=true
    bUseGravityOverride=true
    gravityScale=0.1
    simulationSpace=World
    alphaBoolBlend {blendOption=Linear}
    boundBone {boneName=dyn_shoes_l}
    componentAppliedLinearAccClamp=<10000 10000 10000>
    componentLinearAccScale=<1 1 1>
    componentLinearVelScale=<1 1 1>
    gravityOverride {z=-980}
    physicsBodyDefinitions [{
      boundBone {boneName=dyn_shoes_l}
      boxExtents=<1 1 1>
      constraintSetup {angularLimitsMax=<0 20 0>, angularLimitsMin=<0 -10 0>, angularTarget=<0 0 0>}
      localJointOffset=<0 0 0>
     }]
   }
   Save cached pose 'PriSim'(Pose=Component To Local(ComponentPose=@n15)) {cacheName=PriSim}
   n16 = Control Rig!(
    Source=Use cached pose 'PriSim'()
    Alpha
    isCrouchingCR:double=Get isCrouchBP() -> (isCrouchBP:double)
    isMove:double=Get isMoveBP() -> (isMoveBP:double)
    DBNO:bool=Get bIsDBNO()
    skydive:bool=Get bIsSkyDiving()
    isLod2:bool=Get is LOD2() -> (is LOD2:bool)
   ) -> (Pose) {
    alphaBoolBlend {blendOption=Linear}
    controlRigAssetReference {blueprintRigClass=&M_MED_SmokeyAlias_CtrlRig_C}
   }
   n17 = "Constraint - Base_shoes_l (PPPPP)"!(
    ComponentPose=Local To Component(LocalPose=@n16)
    Alpha=1
    ConstraintWeights_0=0.0008
    ConstraintWeights_1:float=0.0001
    ConstraintWeights_2:float=0.9991
    ConstraintWeights_3:float=0
    ConstraintWeights_4:float=0
   ) -> (Pose) {
    alphaBoolBlend {blendOption=Linear}
    boneToModify {boneName=Base_shoes_l}
    constraintSetup [
     {targetBone {boneName=foot_l}}
     {targetBone {boneName=calf_twist_02_l}}
     {targetBone {boneName=calf_twist_01_l}}
     {targetBone {boneName=calf_twist_02_r}}
     {targetBone {boneName=calf_twist_01_r}}
    ]
    constraintWeights [1,1,1,1,1]
   }
   n18 = "Constraint - Base_shoes_r (PP)"!(ComponentPose=@n17, Alpha=1, ConstraintWeights_0=0.0061, ConstraintWeights_1:float=0.9939) -> (Pose) {
    alphaBoolBlend {blendOption=Linear}
    boneToModify {boneName=Base_shoes_r}
    constraintSetup [{targetBone {boneName=foot_r}}, {targetBone {boneName=calf_twist_01_r}}]
    constraintWeights [1,1]
   }
   n19 = "Transform (Modify) Bone - Bone: Belt_front_r_strap"(ComponentPose=@n18, Alpha=1, Translation="0,1,0", Rotation="0,0,0", Scale="1,1,1") {
    rotationMode=BMM_Additive
    scaleMode=BMM_Additive
    translationMode=BMM_Additive
    alphaBoolBlend {blendOption=Linear}
    boneToModify {boneName=Belt_front_r_strap}
    translation {y=1}
   }
   Save cached pose 'presim'(Pose=Component To Local(ComponentPose=@n19)) {cacheName=presim}
  ]}
 EventGraph {nodes [
   Event BlueprintUpdateAnimation()
   > n1 = Sequence()
   > n2 = Gravity Override State Function DEPRECATED(
    CharacterPartAnimInstance=Self-Reference()
    GameTimeInSeconds=GetGameTimeInSeconds()
    GravityOverrideParams=Get Gravity Override Params 0() -> (Gravity Override Params 0:Gravity Override Params Struct Structure)
   )
   > Set Gravity Override(Gravity Override=@n2)
   Set isCrouchBP(execute=@n1.then_1, isCrouchBP:double="To Float (Boolean)"(InBool=Get bIsCrouching())) -> (Output_Get:double)
   > Set isMoveBP(isMoveBP:double="To Float (Boolean)"(InBool=Get bIsMoving2D())) -> (Output_Get:double)
   > Set is LOD2(is LOD2:bool="integer >= integer"(A=Get CurrentLOD(), B=2)) -> (Output_Get:bool)
   n3 = Get bIsPlayingEmote()
   > Set is my front end(is my front end:bool=AND Boolean(A=NOT Boolean(A=@n3), B=Get bIsInFrontEnd())) -> (Output_Get:bool)
  ]}
]
variables [
 Gravity Override {type=Vector}
 Gravity Override Params 0 {type=GravityOverrideParamsStruct}
 is LOD2 {type=boolean}
 is my front end {type=boolean}
 isCrouchBP {type=number}
 isMoveBP {type=number}
]
