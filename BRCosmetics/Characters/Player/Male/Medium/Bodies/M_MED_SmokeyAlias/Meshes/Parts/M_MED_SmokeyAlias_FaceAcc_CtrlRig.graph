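# Control Rig for the SmokeyAlias face accessory that slightly stretches the Glass bone as the head rotates relative to the neck.
# The forward solve converts the head's rotation relative to neck_02 to Euler angles and takes the X component.
# Two remaps turn positive and negative angles into weights for chained Modify Transforms nodes, each applying a different X-scale, giving an asymmetric correction.
# The hierarchy also carries dyn_Hair bones, which this graph leaves untouched.
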ControlRigBlueprint /BRCosmetics/Characters/Player/Male/Medium/Bodies/M_MED_SmokeyAlias/Meshes/Parts/M_MED_SmokeyAlias_FaceAcc_CtrlRig
graphs [
 RigVMFunctionLibrary {links [], nodes []}
 RigVMModel {nodes [
   RigUnit_BeginExecution = &RigUnit_BeginExecution()
   n1 = &RigVMFunction_MathQuaternionToEuler(
    RotationOrder=ZYX
    Value=&RigUnit_GetRelativeTransformForItem(Child=(Type=Bone,Name="head"), Parent=(Type=Bone,Name="neck_02"), bChildInitial=False, bParentInitial=False).RelativeTransform.Rotation
   )
   &RigUnit_ModifyTransforms(
    ItemToModify=((Item=(Type=Bone,Name="None"),Transform=(Rotation=(X=0,Y=0,Z=0,W=1),Translation=(X=0,Y=0,Z=0),Scale3D=(X=1,Y=1,Z=1))),(Item=(Type=Bone,Name="Glass"),Transform=(Rotation=(X=0,Y=0,Z=0,W=1),Translation=(X=0,Y=0,Z=0),Scale3D=(X=1.04,Y=1,Z=1))))
    WeightMinimum=0
    Weight=&RigVMFunction_MathDoubleRemap(SourceMaximum=60, SourceMinimum=0, TargetMaximum=1, TargetMinimum=0, Value=@n1.Result.X).Result
    ExecutePin=@RigUnit_BeginExecution.ExecutePin
   )
   > &RigUnit_ModifyTransforms(
    ItemToModify=((Item=(Type=Bone,Name="None"),Transform=(Rotation=(X=0,Y=0,Z=0,W=1),Translation=(X=0,Y=0,Z=0),Scale3D=(X=1,Y=1,Z=1))),(Item=(Type=Bone,Name="Glass"),Transform=(Rotation=(X=0,Y=0,Z=0,W=1),Translation=(X=0,Y=0,Z=0),Scale3D=(X=1.06,Y=1,Z=1))))
    WeightMinimum=0
    Weight=&RigVMFunction_MathDoubleRemap(SourceMaximum=-60, SourceMinimum=0, TargetMaximum=1, TargetMinimum=0, Value=@n1.Result.X).Result
   )
  ]}
]
hierarchy [
 root {loc=<0 0 0>, rot=<0 0 0 1>}
 pelvis {parent=root, loc=<0 -0.7071 93.1191>, rot=<0.0127 0.707 -0.0127 -0.707>}
 spine_01 {parent=pelvis, loc=<6.7494 -0.6388 0>, rot=<0 0 0.034 -0.9994>}
 spine_02 {parent=spine_01, loc=<7.2518 -0.035 0>, rot=<0 0 0.0238 -0.9997>}
 spine_03 {parent=spine_02, loc=<6.0104 0 0>, rot=<0 0 0.0149 -0.9999>}
 spine_04 {parent=spine_03, loc=<8.1024 0 0>, rot=<0 0 -0.0132 -0.9999>}
 spine_05 {parent=spine_04, loc=<10.0146 0 0>, rot=<0 0 -0.0015 -1>}
 neck_01 {parent=spine_05, loc=<11.3138 0.1529 0>, rot=<0 0 -0.2003 -0.9797>}
 neck_02 {parent=neck_01, loc=<4.5627 0 0>, rot=<0 0 0.0504 -0.9987>}
 head {parent=neck_02, loc=<5.3914 0 0>, rot=<0 0 0.115 -0.9934>}
 dyn_Hair_21 {parent=head, loc=<9.7642 9.4236 -3.6195>, rot=<-0.1183 -0.479 0.8289 0.2636>}
 dyn_Hair_20 {parent=dyn_Hair_21, loc=<2.5303 -0.0158 0.1694>, rot=<-0.0073 0.2866 0.0245 -0.9577>}
 dyn_Hair_19 {parent=dyn_Hair_20, loc=<2.5651 0 0>, rot=<0 0 0 1>}
 dyn_Hair_18 {parent=head, loc=<10.3718 9.864 -0.4827>, rot=<-0.0607 -0.718 0.6355 0.2774>}
 dyn_Hair_17 {parent=dyn_Hair_18, loc=<2.7117 -0.0416 0.2222>, rot=<-0.0145 0.1415 0.0483 -0.9887>}
 dyn_Hair_16 {parent=dyn_Hair_17, loc=<3.0099 0 0>, rot=<-0.034 0.1258 0.038 -0.9907>}
 dyn_Hair_15 {parent=head, loc=<9.5255 9.1847 3.6863>, rot=<-0.0614 -0.8307 0.5084 0.2184>}
 dyn_Hair_14 {parent=dyn_Hair_15, loc=<2.3343 0 0>, rot=<0 0.1252 0 -0.9921>}
 dyn_Hair_13 {parent=head, loc=<9.3633 7.4206 5.3069>, rot=<0.019 -0.9463 0.2725 0.1728>}
 dyn_Hair_12 {parent=dyn_Hair_13, loc=<2.1441 0 0>, rot=<0.1451 -0.0875 0.0211 -0.9853>}
 dyn_Hair_11 {parent=dyn_Hair_12, loc=<2.2876 0 0>, rot=<0.0737 -0.0021 0.1422 -0.9871>}
 dyn_Hair_10 {parent=head, loc=<7.0683 5.2722 6.4695>, rot=<0.0075 -0.9991 0.0363 -0.0211>}
 dyn_Hair_09 {parent=dyn_Hair_10, loc=<4.5611 0 0>, rot=<0 0 0 1>}
 dyn_Hair_08 {parent=dyn_Hair_09, loc=<4.6887 0 0>, rot=<0 0 0 1>}
 dyn_Hair_29 {parent=head, loc=<7.306 4.7744 -6.5907>, rot=<-0.0034 0.0018 0.9997 -0.0239>}
 dyn_Hair_28 {parent=dyn_Hair_29, loc=<4.9753 -0.0122 -0.0342>, rot=<0.1543 0.0403 -0.0075 -0.9872>}
 dyn_Hair_27 {parent=dyn_Hair_28, loc=<4.942 0 0>, rot=<0 0 0 1>}
 dyn_Hair_24 {parent=head, loc=<9.3426 8.1014 -5.0087>, rot=<-0.2147 -0.4011 0.8783 0.1471>}
 dyn_Hair_23 {parent=dyn_Hair_24, loc=<1.9196 -0.0957 -0.2709>, rot=<0.1484 0.1638 -0.0029 -0.9753>}
 dyn_Hair_22 {parent=dyn_Hair_23, loc=<1.9462 0 0>, rot=<-0.1795 0.1114 0.0205 -0.9772>}
 dyn_Hair_26 {parent=head, loc=<8.9136 6.4715 -6.1415>, rot=<-0.2901 -0.1143 0.9289 0.1997>}
 dyn_Hair_25 {parent=dyn_Hair_26, loc=<2.6537 -0.0756 -0.214>, rot=<0.0859 0.1656 -0.0681 -0.9801>}
 dyn_Hair_07 {parent=head, loc=<14.5921 -4.8936 -0.3866>, rot=<-0.56 0.3323 0.7568 -0.0577>}
 dyn_Hair_06 {parent=dyn_Hair_07, loc=<-0.218 1.5379 0.3497>, rot=<0.7244 0.6875 0.0203 -0.0467>}
 dyn_Hair_05 {parent=dyn_Hair_07, loc=<-0.2064 -1.864 0.0363>, rot=<-0.0179 -0.0444 0.7245 -0.6876>}
 dyn_Hair_04 {parent=dyn_Hair_05, loc=<3.161 -0.4908 -1.4094>, rot=<-0.0135 -0.146 0.0955 -0.9846>}
 dyn_Hair_03 {parent=dyn_Hair_04, loc=<2.4289 1.7682 -1.4716>, rot=<-0.2174 -0.3004 0.0705 -0.926>}
 dyn_Hair_02 {parent=dyn_Hair_04, loc=<1.5599 -1.4181 -3.0487>, rot=<0.0442 -0.5824 0.0945 -0.8062>}
 dyn_Hair_01 {parent=dyn_Hair_04, loc=<2.662 -0.8683 1.2883>, rot=<0.934 -0.1478 -0.2885 -0.1499>}
 Glass {parent=head, loc=<4.2416 10.4053 0>, rot=<-0.0029 0.7071 -0.0029 -0.7071>}
]
