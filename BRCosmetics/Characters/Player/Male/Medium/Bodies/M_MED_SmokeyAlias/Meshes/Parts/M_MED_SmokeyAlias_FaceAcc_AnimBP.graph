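# Post-pose Animation Blueprint for the SmokeyAlias face accessory cosmetic part, adding secondary motion to its hair.
# It copies the attached parent mesh's pose, runs the part's Control Rig, then nudges the Glass bone additively.
# Chained Anim Dynamics nodes spring dyn_Hair bones, mostly within head-relative spherical limits; some use gravity, others none.
# The event graph updates a Gravity Override variable via a deprecated function, but the dynamics nodes show fixed gravity values.
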AnimBlueprint /BRCosmetics/Characters/Player/Male/Medium/Bodies/M_MED_SmokeyAlias/Meshes/Parts/M_MED_SmokeyAlias_FaceAcc_AnimBP
target_skeleton=&/M_MED_SmokeyAlias_FaceAcc_Skeleton
graphs [
 AnimGraph {nodes [
   n1 = Copy Pose From Mesh() {bUseAttachedParent=true}
   n2 = Control Rig(Source=Component To Local(ComponentPose=Local To Component(LocalPose=@n1))) {
    alphaBoolBlend {blendOption=Linear}
    controlRigAssetReference {blueprintRigClass=&/M_MED_SmokeyAlias_FaceAcc_CtrlRig.M_MED_SmokeyAlias_FaceAcc_CtrlRig_C}
   }
   n3 = "Transform (Modify) Bone - Bone: Glass"(ComponentPose=Local To Component(LocalPose=@n2), Alpha=1, Translation="0,0.3,0", Rotation="0,0,0", Scale="1,1,1") {
    rotationMode=BMM_Additive
    scaleMode=BMM_Additive
    translationMode=BMM_Additive
    alphaBoolBlend {blendOption=Linear}
    boneToModify {boneName=Glass}
    translation {y=0.3}
   }
   n4 = Anim Dynamics - Bone: dyn_Hair_10!(ComponentPose=@n3, Alpha=1, GravityOverride:Vector="0,0,-980", ExternalForce="0,0,0") -> (Pose) {
    angularDampingOverride=0.9999
    angularSpringConstant=10
    bAngularSpring=true
    bOverrideAngularDamping=true
    bUseGravityOverride=true
    linearDampingOverride=0.99
    alphaBoolBlend {blendOption=Linear}
    boundBone {boneName=dyn_Hair_10}
    gravityOverride {z=-980}
    physicsBodyDefinitions [{
      boundBone {boneName=dyn_Hair_10}
      boxExtents=<7.5 7.5 7.5>
      constraintSetup {angularLimitsMax=<10 30 17>, angularLimitsMin=<-10 -3 -17>, angularTarget=<0 0 0>}
      localJointOffset=<1 0 0>
     }]
    relativeSpaceBone {boneName=head}
    sphericalLimits [{limitRadius=7.8476, drivingBone {boneName=head}, sphereLocalOffset=<4.5172 2.1965 -1.6179>}]
   }
   n5 = Anim Dynamics - Bone: dyn_Hair_09!(ComponentPose=@n4, Alpha=1, GravityOverride:Vector="0,0,-980", ExternalForce="0,0,0") -> (Pose) {
    angularDampingOverride=0.9999
    angularSpringConstant=10
    bAngularSpring=true
    bOverrideAngularDamping=true
    bUseGravityOverride=true
    linearDampingOverride=0.99
    alphaBoolBlend {blendOption=Linear}
    boundBone {boneName=dyn_Hair_09}
    gravityOverride {z=-980}
    physicsBodyDefinitions [{
      boundBone {boneName=dyn_Hair_09}
      boxExtents=<7 7 7>
      constraintSetup {angularLimitsMax=<10 15 14>, angularLimitsMin=<-10 0 -14>, angularTarget=<0 0 0>}
      localJointOffset=<1 0 0>
     }]
    relativeSpaceBone {boneName=head}
    sphericalLimits [{limitRadius=10.6793, drivingBone {boneName=head}, sphereLocalOffset=<3.0448 3.7379 -4.908>}]
   }
   n6 = Anim Dynamics - Bone: dyn_Hair_08!(ComponentPose=@n5, Alpha=1, GravityOverride:Vector="0,0,-980", ExternalForce="0,0,0") -> (Pose) {
    angularDampingOverride=0.9999
    angularSpringConstant=10
    bAngularSpring=true
    bOverrideAngularDamping=true
    bUseGravityOverride=true
    linearDampingOverride=0.99
    alphaBoolBlend {blendOption=Linear}
    boundBone {boneName=dyn_Hair_08}
    gravityOverride {z=-980}
    physicsBodyDefinitions [{
      boundBone {boneName=dyn_Hair_08}
      boxExtents=<6 6 6>
      constraintSetup {angularLimitsMax=<10 10 10>, angularLimitsMin=<-10 10 -10>, angularTarget=<0 0 0>}
      localJointOffset=<1 0 0>
     }]
    relativeSpaceBone {boneName=head}
    sphericalLimits [{limitRadius=10.6793, drivingBone {boneName=head}, sphereLocalOffset=<-2.206 3.5444 -5.6533>}]
   }
   n7 = Anim Dynamics - Bone: dyn_Hair_13!(ComponentPose=@n6, Alpha=1, GravityOverride:Vector="0,0,0", ExternalForce="0,0,0") -> (Pose) {
    angularDampingOverride=0.99
    angularSpringConstant=10
    bAngularSpring=true
    bOverrideAngularDamping=true
    bUseGravityOverride=true
    linearDampingOverride=0.99
    alphaBoolBlend {blendOption=Linear}
    boundBone {boneName=dyn_Hair_13}
    physicsBodyDefinitions [{
      boundBone {boneName=dyn_Hair_13}
      boxExtents=<9 9 9>
      constraintSetup {angularLimitsMax=<10 15 15>, angularLimitsMin=<-10 -1 13>, angularTarget=<0 0 0>}
      localJointOffset=<1 0 0>
     }]
    relativeSpaceBone {boneName=head}
    sphericalLimits [{limitRadius=10.6793, drivingBone {boneName=head}, sphereLocalOffset=<3.0448 3.5444 -3.4816>}]
   }
   n8 = Anim Dynamics - Bone: dyn_Hair_12!(ComponentPose=@n7, Alpha=1, GravityOverride:Vector="0,0,0", ExternalForce="0,0,0") -> (Pose) {
    angularDampingOverride=0.99
    angularSpringConstant=10
    bAngularSpring=true
    bOverrideAngularDamping=true
    bUseGravityOverride=true
    linearDampingOverride=0.99
    alphaBoolBlend {blendOption=Linear}
    boundBone {boneName=dyn_Hair_12}
    physicsBodyDefinitions [{
      boundBone {boneName=dyn_Hair_12}
      boxExtents=<12 12 12>
      constraintSetup {angularLimitsMax=<10 15 10>, angularLimitsMin=<-10 -5 -10>, angularTarget=<0 0 0>}
      localJointOffset=<1 0 0>
     }]
    relativeSpaceBone {boneName=head}
    sphericalLimits [{limitRadius=10.6793, drivingBone {boneName=head}, sphereLocalOffset=<3.0448 3.5444 -3.4816>}]
   }
   n9 = Anim Dynamics - Bone: dyn_Hair_15!(ComponentPose=@n8, Alpha=1, GravityOverride:Vector="0,0,0", ExternalForce="0,0,0") -> (Pose) {
    angularDampingOverride=0.99
    angularSpringConstant=10
    bAngularSpring=true
    bOverrideAngularDamping=true
    bUseGravityOverride=true
    linearDampingOverride=0.99
    alphaBoolBlend {blendOption=Linear}
    boundBone {boneName=dyn_Hair_15}
    physicsBodyDefinitions [{
      boundBone {boneName=dyn_Hair_15}
      boxExtents=<12 12 12>
      constraintSetup {angularLimitsMax=<10 15 10>, angularLimitsMin=<-10 -1 -10>, angularTarget=<0 0 0>}
      localJointOffset=<1 0 0>
     }]
    relativeSpaceBone {boneName=head}
    sphericalLimits [{limitRadius=10.6793, drivingBone {boneName=head}, sphereLocalOffset=<3.0448 3.5444 -3.4816>}]
   }
   n10 = Anim Dynamics - Bone: dyn_Hair_18!(ComponentPose=@n9, Alpha=1, GravityOverride:Vector="0,0,0", ExternalForce="0,0,0") -> (Pose) {
    angularDampingOverride=0.99
    angularSpringConstant=10
    bAngularSpring=true
    bOverrideAngularDamping=true
    bUseGravityOverride=true
    linearDampingOverride=0.99
    alphaBoolBlend {blendOption=Linear}
    boundBone {boneName=dyn_Hair_18}
    physicsBodyDefinitions [{
      boundBone {boneName=dyn_Hair_18}
      boxExtents=<12 12 12>
      constraintSetup {angularLimitsMax=<10 13 7>, angularLimitsMin=<-10 -1 -10>, angularTarget=<0 0 0>}
      localJointOffset=<1 0 0>
     }]
    relativeSpaceBone {boneName=head}
    sphericalLimits [{limitRadius=10.6793, drivingBone {boneName=head}, sphereLocalOffset=<3.0448 1.9736 -3.4816>}]
   }
   n11 = Anim Dynamics - Bone: dyn_Hair_17!(ComponentPose=@n10, Alpha=1, GravityOverride:Vector="0,0,0", ExternalForce="0,0,0") -> (Pose) {
    angularDampingOverride=0.99
    angularSpringConstant=10
    bAngularSpring=true
    bOverrideAngularDamping=true
    bUseGravityOverride=true
    linearDampingOverride=0.99
    alphaBoolBlend {blendOption=Linear}
    boundBone {boneName=dyn_Hair_17}
    physicsBodyDefinitions [{
      boundBone {boneName=dyn_Hair_17}
      boxExtents=<12 12 12>
      constraintSetup {angularLimitsMax=<10 12 7>, angularLimitsMin=<-10 -1 -10>, angularTarget=<0 0 0>}
      localJointOffset=<1 0 0>
     }]
    relativeSpaceBone {boneName=head}
    sphericalLimits [{limitRadius=10.6793, drivingBone {boneName=head}, sphereLocalOffset=<2.2353 1.9736 -3.4816>}]
   }
   n12 = Anim Dynamics - Bone: dyn_Hair_21!(ComponentPose=@n11, Alpha=1, GravityOverride:Vector="0,0,0", ExternalForce="0,0,0") -> (Pose) {
    angularDampingOverride=0.99
    angularSpringConstant=10
    bAngularSpring=true
    bOverrideAngularDamping=true
    bUseGravityOverride=true
    linearDampingOverride=0.99
    alphaBoolBlend {blendOption=Linear}
    boundBone {boneName=dyn_Hair_21}
    physicsBodyDefinitions [{
      boundBone {boneName=dyn_Hair_21}
      boxExtents=<12 12 12>
      constraintSetup {angularLimitsMax=<10 15 13>, angularLimitsMin=<-10 -1 -13>, angularTarget=<0 0 0>}
      localJointOffset=<1 0 0>
     }]
    relativeSpaceBone {boneName=head}
    sphericalLimits [{limitRadius=10.6793, drivingBone {boneName=head}, sphereLocalOffset=<2.6711 1.2092 -3.4816>}]
   }
   n13 = Anim Dynamics - Bone: dyn_Hair_20!(ComponentPose=@n12, Alpha=1, GravityOverride:Vector="0,0,0", ExternalForce="0,0,0") -> (Pose) {
    angularDampingOverride=0.99
    angularSpringConstant=10
    bAngularSpring=true
    bOverrideAngularDamping=true
    bUseGravityOverride=true
    linearDampingOverride=0.99
    alphaBoolBlend {blendOption=Linear}
    boundBone {boneName=dyn_Hair_20}
    physicsBodyDefinitions [{
      boundBone {boneName=dyn_Hair_20}
      boxExtents=<12 12 12>
      constraintSetup {angularLimitsMax=<10 10 5>, angularLimitsMin=<-10 -1 -10>, angularTarget=<0 0 0>}
      localJointOffset=<1 0 0>
     }]
    relativeSpaceBone {boneName=head}
    sphericalLimits [{limitRadius=10.6793, drivingBone {boneName=head}, sphereLocalOffset=<2.6711 1.2092 -3.4816>}]
   }
   n14 = Anim Dynamics - Bone: dyn_Hair_24!(ComponentPose=@n13, Alpha=1, GravityOverride:Vector="0,0,0", ExternalForce="0,0,0") -> (Pose) {
    angularDampingOverride=0.99
    angularSpringConstant=10
    bAngularSpring=true
    bOverrideAngularDamping=true
    bUseGravityOverride=true
    linearDampingOverride=0.99
    alphaBoolBlend {blendOption=Linear}
    boundBone {boneName=dyn_Hair_24}
    physicsBodyDefinitions [{
      boundBone {boneName=dyn_Hair_24}
      boxExtents=<12 12 12>
      constraintSetup {angularLimitsMax=<10 10 5>, angularLimitsMin=<-10 -1 -10>, angularTarget=<0 0 0>}
      localJointOffset=<1 0 0>
     }]
    relativeSpaceBone {boneName=head}
    sphericalLimits [{limitRadius=10.6793, drivingBone {boneName=head}, sphereLocalOffset=<2.6711 0.9499 -3.4816>}]
   }
   n15 = Anim Dynamics - Bone: dyn_Hair_26!(ComponentPose=@n14, Alpha=1, GravityOverride:Vector="0,0,0", ExternalForce="0,0,0") -> (Pose) {
    angularDampingOverride=0.99
    angularSpringConstant=10
    bAngularSpring=true
    bOverrideAngularDamping=true
    bUseGravityOverride=true
    linearDampingOverride=0.99
    alphaBoolBlend {blendOption=Linear}
    boundBone {boneName=dyn_Hair_26}
    physicsBodyDefinitions [{
      boundBone {boneName=dyn_Hair_26}
      boxExtents=<12 12 12>
      constraintSetup {angularLimitsMax=<10 10 5>, angularLimitsMin=<-10 -1 -10>, angularTarget=<0 0 0>}
      localJointOffset=<1 0 0>
     }]
    relativeSpaceBone {boneName=head}
    sphericalLimits [{limitRadius=10.6793, drivingBone {boneName=head}, sphereLocalOffset=<2.6711 0.9499 2.7668>}]
   }
   n16 = Anim Dynamics - Bone: dyn_Hair_29!(ComponentPose=@n15, Alpha=1, GravityOverride:Vector="0,0,0", ExternalForce="0,0,0") -> (Pose) {
    angularDampingOverride=1
    angularSpringConstant=10
    bAngularSpring=true
    bOverrideAngularDamping=true
    bUseGravityOverride=true
    linearDampingOverride=0.99
    alphaBoolBlend {blendOption=Linear}
    boundBone {boneName=dyn_Hair_29}
    physicsBodyDefinitions [{
      boundBone {boneName=dyn_Hair_29}
      boxExtents=<7.5 7.5 7.5>
      constraintSetup {angularLimitsMax=<10 30 10>, angularLimitsMin=<-10 -1 -10>, angularTarget=<0 0 0>}
      localJointOffset=<1 0 0>
     }]
    relativeSpaceBone {boneName=head}
    sphericalLimits [{limitRadius=10.6793, drivingBone {boneName=head}, sphereLocalOffset=<2.6711 4.3881 3.677>}]
   }
   n17 = Anim Dynamics - Bone: dyn_Hair_28!(ComponentPose=@n16, Alpha=1, GravityOverride:Vector="0,0,0", ExternalForce="0,0,0") -> (Pose) {
    angularDampingOverride=1
    angularSpringConstant=10
    bAngularSpring=true
    bOverrideAngularDamping=true
    bUseGravityOverride=true
    linearDampingOverride=0.99
    alphaBoolBlend {blendOption=Linear}
    boundBone {boneName=dyn_Hair_28}
    physicsBodyDefinitions [{
      boundBone {boneName=dyn_Hair_28}
      boxExtents=<7 7 7>
      constraintSetup {angularLimitsMax=<10 15 13>, angularLimitsMin=<-10 -1 -13>, angularTarget=<0 0 0>}
      localJointOffset=<1 0 0>
     }]
    relativeSpaceBone {boneName=head}
    sphericalLimits [{limitRadius=10.6793, drivingBone {boneName=head}, sphereLocalOffset=<1.3309 0.9499 4.1765>}]
   }
   n18 = Anim Dynamics - Bone: dyn_Hair_07!(ComponentPose=@n17, Alpha=1, GravityOverride:Vector="0,0,0", ExternalForce="0,0,0") -> (Pose) {
    angularDampingOverride=0.99
    angularSpringConstant=10
    bAngularSpring=true
    bOverrideAngularDamping=true
    bUseGravityOverride=true
    linearDampingOverride=0.99
    alphaBoolBlend {blendOption=Linear}
    boundBone {boneName=dyn_Hair_07}
    physicsBodyDefinitions [{
      boundBone {boneName=dyn_Hair_07}
      boxExtents=<9 9 9>
      constraintSetup {angularLimitsMax=<4 5 2>, angularLimitsMin=<4 -1 -1>, angularTarget=<0 0 0>}
      localJointOffset=<1 0 0>
     }]
    relativeSpaceBone {boneName=head}
    sphericalLimits [{limitRadius=10.6793, drivingBone {boneName=head}, sphereLocalOffset=<2.6711 0.9499 2.7668>}]
   }
   n19 = Anim Dynamics - Bone: dyn_Hair_05!(ComponentPose=@n18, Alpha=1, GravityOverride:Vector="0,0,0", ExternalForce="0,0,0") -> (Pose) {
    angularDampingOverride=0.99
    angularSpringConstant=10
    bAngularSpring=true
    bOverrideAngularDamping=true
    bUseGravityOverride=true
    linearDampingOverride=0.99
    alphaBoolBlend {blendOption=Linear}
    boundBone {boneName=dyn_Hair_05}
    physicsBodyDefinitions [{
      boundBone {boneName=dyn_Hair_05}
      boxExtents=<12 12 12>
      constraintSetup {angularLimitsMax=<10 10 17>, angularLimitsMin=<-10 -1 -17>, angularTarget=<0 0 0>}
      localJointOffset=<1 0 0>
     }]
    relativeSpaceBone {boneName=head}
    sphericalLimits [{limitRadius=10.6793, drivingBone {boneName=head}, sphereLocalOffset=<2.6711 0.9499 2.7668>}]
   }
   n20 = Anim Dynamics - Bone: dyn_Hair_06!(ComponentPose=@n19, Alpha=1, GravityOverride:Vector="0,0,-980", ExternalForce="0,0,0") -> (Pose) {
    angularDampingOverride=1
    angularSpringConstant=1500
    bAngularSpring=true
    bGravityOverrideInSimSpace=true
    bOverrideAngularDamping=true
    bUseGravityOverride=true
    gravityScale=0.9
    simulationSpace=RootRelative
    alphaBoolBlend {blendOption=Linear}
    boundBone {boneName=dyn_Hair_06}
    gravityOverride {z=-980}
    physicsBodyDefinitions [{
      sphereCollisionRadius=0
      boundBone {boneName=dyn_Hair_06}
      boxExtents=<10 10 10>
      constraintSetup {angularTargetAxis=AxisY, angularLimitsMax=<10 10 10>, angularLimitsMin=<-10 -10 -10>, angularTarget=<0 0 0>}
      localJointOffset=<2 0 0>
     }]
    relativeSpaceBone {boneName=root}
   }
   Output Pose(Result=Component To Local(ComponentPose=@n20)) {layerGroup=None}
  ]}
 EventGraph {nodes [
   Event BlueprintUpdateAnimation()
   > n1 = Gravity Override State Function DEPRECATED(
    CharacterPartAnimInstance=Self-Reference()
    GameTimeInSeconds=GetGameTimeInSeconds()
    GravityOverrideParams=Get Gravity Override Params 0() -> (Gravity Override Params 0:Gravity Override Params Struct Structure)
   )
   > Set Gravity Override(Gravity Override=@n1)
  ]}
]
variables [
 Gravity Override {type=Vector}
 Gravity Override Params 0 {type=GravityOverrideParamsStruct, default {onGround_Moving {jointName=head, gravityOverride=<0 0 0>}}}
]
